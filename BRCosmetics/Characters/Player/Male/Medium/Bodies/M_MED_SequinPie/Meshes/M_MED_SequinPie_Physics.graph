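# Physics asset for the SequinPie male medium character body, setting up dynamic motion for a skirt and two pouch ties.
# Capsules on pelvis and legs, plus a large floor sphere, serve as body colliders.
# Six skirt chains and both tie chains hang from the pelvis with mostly limited angular joints.
# Angular-free cross-links join neighboring skirt panels mid-chain, except front-to-front; ties link to front panels, angular-free left and angular-locked right.
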PhysicsAsset /BRCosmetics/Characters/Player/Male/Medium/Bodies/M_MED_SequinPie/Meshes/M_MED_SequinPie_Physics
bodies {
 calf_l [{shapeType=Capsule, length=36.8728, radius=10.146, center=<-22.6006 -0.091 1.594>, rotation=<rot 0 90.2306 -88.6141>}]
 calf_r [{shapeType=Capsule, length=36.8728, radius=10.146, center=<22.6006 0.091 -1.594>, rotation=<rot 0 90.2306 91.3859>}]
 dyn_floor_collision [{shapeType=Sphere, radius=300, shapeName=dyn_floor_collision_sphere, center=<0 0 -280>}]
 dyn_pouch_tie_l_01 [{
   shapeType=Capsule
   length=0.9559
   radius=0.952
   shapeName=dyn_pouch_tie_l_01_capsule
   center=<1.0855 0.0349 -0.0338>
   rotation=<rot 0 -5.1789 -67.6658>
  }]
 dyn_pouch_tie_l_02 [{
   shapeType=Capsule
   length=7.063
   radius=3.5665
   shapeName=dyn_pouch_tie_l_02_capsule
   center=<3.712 0.9182 0.8512>
   rotation=<rot 2.1004 89.5035 -81.873>
  }]
 dyn_pouch_tie_l_MSTR [{shapeType=Capsule, length=0.505, radius=0.505, shapeName=dyn_pouch_tie_l_MSTR_capsule, rotation=<rot 0 -90 0>}]
 dyn_pouch_tie_r_01 [{
   shapeType=Capsule
   length=0.9559
   radius=0.952
   shapeName=dyn_pouch_tie_l_01_capsule
   center=<-1.0855 -0.0349 0.0338>
   rotation=<rot 0 -5.1789 112.3342>
  }]
 dyn_pouch_tie_r_02 [{
   shapeType=Capsule
   length=7.063
   radius=3.5665
   shapeName=dyn_pouch_tie_l_02_capsule
   center=<-3.7027 0.1543 -0.8118>
   rotation=<rot 1.5006 93.6699 98.2579>
  }]
 dyn_pouch_tie_r_MSTR [{shapeType=Capsule, length=0.505, radius=0.505, shapeName=dyn_pouch_tie_l_MSTR_capsule, rotation=<rot 0 -90 180>}]
 dyn_skirt_front_l_01 [{
   shapeType=Capsule
   length=9.6938
   radius=4.9
   shapeName=dyn_skirt_front_l_01_capsule
   center=<10 -2.4816 6.094>
   rotation=<rot 0.0003 0 29.9999>
  }]
 dyn_skirt_front_l_02 [{
   shapeType=Capsule
   length=14.1
   radius=5.9
   shapeName=dyn_skirt_front_l_02_capsule
   center=<3.9992 -3.7977 5.6071>
   rotation=<rot 0.0078 0 29.3668>
  }]
 dyn_skirt_front_l_03 [{shapeType=Capsule, length=23.8063, radius=7.6, shapeName=dyn_skirt_front_l_03_capsule, center=<4 -6 0>, rotation=<rot 0.0001 0 26.557>}]
 dyn_skirt_front_l_04 [{shapeType=Capsule, length=23.4062, radius=8, shapeName=dyn_skirt_front_l_04_capsule, center=<4 -6 0>, rotation=<rot 0.0001 0 30.8245>}]
 dyn_skirt_front_l_05 [{shapeType=Capsule, length=22.2062, radius=9.2, shapeName=dyn_skirt_front_l_05_capsule, center=<4 -6 0>, rotation=<rot 0.0001 0 31.975>}]
 dyn_skirt_front_l_MSTR [{shapeType=Capsule, length=1, radius=1, shapeName=dyn_skirt_front_l_MSTR_capsule}]
 dyn_skirt_front_r_01 [{
   shapeType=Capsule
   length=9.6938
   radius=4.9
   shapeName=dyn_skirt_front_l_01_capsule
   center=<-10 2.4816 -6.094>
   rotation=<rot 0.0003 0 -150.0001>
  }]
 dyn_skirt_front_r_02 [{
   shapeType=Capsule
   length=14.1
   radius=5.9
   shapeName=dyn_skirt_front_l_02_capsule
   center=<-3.9992 3.7977 -5.6071>
   rotation=<rot 0.0078 0 -150.6332>
  }]
 dyn_skirt_front_r_03 [{shapeType=Capsule, length=23.8063, radius=7.6, shapeName=dyn_skirt_front_l_03_capsule, center=<-4 6 0>, rotation=<rot 0.0001 0 -153.443>}]
 dyn_skirt_front_r_04 [{shapeType=Capsule, length=23.4062, radius=8, shapeName=dyn_skirt_front_l_04_capsule, center=<-4 6 0>, rotation=<rot 0.0001 0 -149.1755>}]
 dyn_skirt_front_r_05 [{shapeType=Capsule, length=22.2062, radius=9.2, shapeName=dyn_skirt_front_l_05_capsule, center=<-4 6 0>, rotation=<rot 0.0001 0 -148.025>}]
 dyn_skirt_front_r_MSTR [{shapeType=Capsule, length=1, radius=1, shapeName=dyn_skirt_front_l_MSTR_capsule}]
 dyn_skirt_rear_l_01 [{shapeType=Capsule, length=12, radius=8, shapeName=dyn_skirt_rear_l_01_capsule, center=<10 -6 0>}]
 dyn_skirt_rear_l_02 [{shapeType=Capsule, length=12, radius=8, shapeName=dyn_skirt_rear_l_02_capsule, center=<4 -6 0>}]
 dyn_skirt_rear_l_03 [{shapeType=Capsule, length=15.5331, radius=10.5669, shapeName=dyn_skirt_rear_l_03_capsule, center=<4 -6 0>}]
 dyn_skirt_rear_l_04 [{shapeType=Capsule, length=20.5544, radius=11.5456, shapeName=dyn_skirt_rear_l_04_capsule, center=<4 -6 0>}]
 dyn_skirt_rear_l_05 [{shapeType=Capsule, length=24, radius=8.7, shapeName=dyn_skirt_rear_l_05_capsule, center=<4 -6 0>}]
 dyn_skirt_rear_l_MSTR [{shapeType=Capsule, length=1, radius=1, shapeName=dyn_skirt_rear_l_MSTR_capsule}]
 dyn_skirt_rear_r_01 [{shapeType=Capsule, length=12, radius=8, shapeName=dyn_skirt_rear_l_01_capsule, center=<-10 6 0>, rotation=<rot 0 0 -180>}]
 dyn_skirt_rear_r_02 [{shapeType=Capsule, length=12, radius=8, shapeName=dyn_skirt_rear_l_02_capsule, center=<-4 6 0>, rotation=<rot 0 0 -180>}]
 dyn_skirt_rear_r_03 [{shapeType=Capsule, length=16.2125, radius=9.8875, shapeName=dyn_skirt_rear_l_03_capsule, center=<-4 6 0>, rotation=<rot 0 0 -180>}]
 dyn_skirt_rear_r_04 [{shapeType=Capsule, length=21.7437, radius=10.3563, shapeName=dyn_skirt_rear_l_04_capsule, center=<-4 6 0>, rotation=<rot 0 0 -180>}]
 dyn_skirt_rear_r_05 [{shapeType=Capsule, length=24, radius=8.7, shapeName=dyn_skirt_rear_l_05_capsule, center=<-4 6 0>, rotation=<rot 0 0 -180>}]
 dyn_skirt_rear_r_MSTR [{shapeType=Capsule, length=1, radius=1, shapeName=dyn_skirt_rear_r_MSTR_capsule}]
 dyn_skirt_side_l_01 [{shapeType=Capsule, length=12, radius=8, shapeName=dyn_skirt_side_l_01_capsule, center=<10 -6 0>}]
 dyn_skirt_side_l_02 [{shapeType=Capsule, length=12, radius=8, shapeName=dyn_skirt_side_l_02_capsule, center=<4 -6 0>}]
 dyn_skirt_side_l_03 [{shapeType=Capsule, length=10.4551, radius=9.5449, shapeName=dyn_skirt_side_l_03_capsule, center=<4 -6 0>}]
 dyn_skirt_side_l_04 [{shapeType=Capsule, length=15.7296, radius=10.3704, shapeName=dyn_skirt_side_l_04_capsule, center=<4 -6 0>}]
 dyn_skirt_side_l_05 [{shapeType=Capsule, length=18.7106, radius=9.1894, shapeName=dyn_skirt_side_l_05_capsule, center=<4 -6 0>}]
 dyn_skirt_side_l_MSTR [{shapeType=Capsule, length=1, radius=1, shapeName=dyn_skirt_side_l_MSTR_capsule}]
 dyn_skirt_side_r_01 [{shapeType=Capsule, length=12, radius=8, shapeName=dyn_skirt_side_l_01_capsule, center=<-10 6 0>, rotation=<rot 0 0 -180>}]
 dyn_skirt_side_r_02 [{shapeType=Capsule, length=12, radius=8, shapeName=dyn_skirt_side_l_02_capsule, center=<-4 6 0>, rotation=<rot 0 0 -180>}]
 dyn_skirt_side_r_03 [{shapeType=Capsule, length=11.2255, radius=8.7745, shapeName=dyn_skirt_side_l_03_capsule, center=<-4 6 0>, rotation=<rot 0 0 -180>}]
 dyn_skirt_side_r_04 [{shapeType=Capsule, length=16.2043, radius=9.8957, shapeName=dyn_skirt_side_l_04_capsule, center=<-4 6 0>, rotation=<rot 0 0 -180>}]
 dyn_skirt_side_r_05 [{shapeType=Capsule, length=19.9, radius=8, shapeName=dyn_skirt_side_l_05_capsule, center=<-4 6 0>, rotation=<rot 0 0 -180>}]
 dyn_skirt_side_r_MSTR [{shapeType=Capsule, length=1, radius=1, shapeName=dyn_skirt_front_l_MSTR_capsule}]
 foot_l [{shapeType=Capsule, length=18.049, radius=8.4077, center=<-2.9676 8.4487 0.5569>, rotation=<rot 0 7.8917 -89.3248>}]
 foot_r [{shapeType=Capsule, length=18.049, radius=8.4077, center=<2.9676 -8.4487 -0.5569>, rotation=<rot 0 7.8917 90.6752>}]
 pelvis [{shapeType=Capsule, length=13, radius=10}]
 thigh_l [{shapeType=Capsule, length=32.6536, radius=8.4421, center=<-18.3578 0.5947 0>, rotation=<rot 15.2652 88.1447 -90>}]
 thigh_r [{shapeType=Capsule, length=32.6536, radius=9.4421, center=<18.3578 -0.5947 0>, rotation=<rot 15.2652 88.1447 90>}]
}
constraints [
 {bone1Name=thigh_l, bone2Name=pelvis, swing1LimitDegrees=45, swing2LimitDegrees=45, twistLimitDegrees=45}
 {bone1Name=calf_l, bone2Name=thigh_l, swing1LimitDegrees=45, swing2LimitDegrees=45, twistLimitDegrees=45}
 {bone1Name=foot_l, bone2Name=calf_l, swing1LimitDegrees=45, swing2LimitDegrees=45, twistLimitDegrees=45}
 {bone1Name=thigh_r, bone2Name=pelvis, swing1LimitDegrees=45, swing2LimitDegrees=45, twistLimitDegrees=45}
 {bone1Name=calf_r, bone2Name=thigh_r, swing1LimitDegrees=45, swing2LimitDegrees=45, twistLimitDegrees=45}
 {bone1Name=foot_r, bone2Name=calf_r, swing1LimitDegrees=45, swing2LimitDegrees=45, twistLimitDegrees=45}
 {bone1Name=dyn_skirt_front_l_MSTR, bone2Name=pelvis, swing1LimitDegrees=12, swing2LimitDegrees=12, twistLimitDegrees=12}
 {bone1Name=dyn_skirt_front_l_01, bone2Name=dyn_skirt_front_l_MSTR, swing1LimitDegrees=24, swing2LimitDegrees=14, twistLimitDegrees=14}
 {bone1Name=dyn_skirt_front_l_02, bone2Name=dyn_skirt_front_l_01, swing1LimitDegrees=24, swing2LimitDegrees=14, twistLimitDegrees=14}
 {bone1Name=dyn_skirt_front_l_03, bone2Name=dyn_skirt_front_l_02, swing1LimitDegrees=24, swing2LimitDegrees=14, twistLimitDegrees=14}
 {bone1Name=dyn_skirt_front_l_04, bone2Name=dyn_skirt_front_l_03, swing1LimitDegrees=24, swing2LimitDegrees=14, twistLimitDegrees=14}
 {bone1Name=dyn_skirt_front_l_05, bone2Name=dyn_skirt_front_l_04, swing1LimitDegrees=8, swing2LimitDegrees=8, twistLimitDegrees=4}
 {bone1Name=dyn_skirt_rear_l_MSTR, bone2Name=pelvis, swing1LimitDegrees=12, swing2LimitDegrees=12, twistLimitDegrees=12}
 {bone1Name=dyn_skirt_rear_l_01, bone2Name=dyn_skirt_rear_l_MSTR, swing1LimitDegrees=24, swing2LimitDegrees=14, twistLimitDegrees=14}
 {bone1Name=dyn_skirt_rear_l_02, bone2Name=dyn_skirt_rear_l_01, swing1LimitDegrees=24, swing2LimitDegrees=14, twistLimitDegrees=14}
 {bone1Name=dyn_skirt_rear_l_03, bone2Name=dyn_skirt_rear_l_02, swing1LimitDegrees=24, swing2LimitDegrees=14, twistLimitDegrees=14}
 {bone1Name=dyn_skirt_rear_l_04, bone2Name=dyn_skirt_rear_l_03, swing1LimitDegrees=24, swing2LimitDegrees=14, twistLimitDegrees=14}
 {bone1Name=dyn_skirt_rear_l_05, bone2Name=dyn_skirt_rear_l_04, swing1LimitDegrees=8, swing2LimitDegrees=8, twistLimitDegrees=4}
 {bone1Name=dyn_skirt_side_l_MSTR, bone2Name=pelvis, swing1LimitDegrees=12, swing2LimitDegrees=12, twistLimitDegrees=12}
 {bone1Name=dyn_skirt_side_l_01, bone2Name=dyn_skirt_side_l_MSTR, swing1LimitDegrees=24, swing2LimitDegrees=14, twistLimitDegrees=14}
 {bone1Name=dyn_skirt_side_l_02, bone2Name=dyn_skirt_side_l_01, swing1LimitDegrees=24, swing2LimitDegrees=14, twistLimitDegrees=14}
 {bone1Name=dyn_skirt_side_l_03, bone2Name=dyn_skirt_side_l_02, swing1LimitDegrees=24, swing2LimitDegrees=14, twistLimitDegrees=14}
 {bone1Name=dyn_skirt_side_l_04, bone2Name=dyn_skirt_side_l_03, swing1LimitDegrees=24, swing2LimitDegrees=14, twistLimitDegrees=14}
 {bone1Name=dyn_skirt_side_l_05, bone2Name=dyn_skirt_side_l_04, swing1LimitDegrees=8, swing2LimitDegrees=8, twistLimitDegrees=4}
 {bone1Name=dyn_skirt_front_r_MSTR, bone2Name=pelvis, swing1LimitDegrees=12, swing2LimitDegrees=12, twistLimitDegrees=12}
 {
  bone1Name=dyn_skirt_rear_r_MSTR
  bone2Name=pelvis
  swing1LimitDegrees=45
  swing1Motion=Free
  swing2LimitDegrees=45
  swing2Motion=Free
  twistLimitDegrees=45
  twistMotion=Free
 }
 {bone1Name=dyn_skirt_side_r_MSTR, bone2Name=pelvis, swing1LimitDegrees=12, swing2LimitDegrees=12, twistLimitDegrees=12}
 {bone1Name=dyn_skirt_rear_r_01, bone2Name=dyn_skirt_rear_r_MSTR, swing1LimitDegrees=24, swing2LimitDegrees=14, twistLimitDegrees=14}
 {bone1Name=dyn_skirt_rear_r_02, bone2Name=dyn_skirt_rear_r_01, swing1LimitDegrees=24, swing2LimitDegrees=14, twistLimitDegrees=14}
 {bone1Name=dyn_skirt_rear_r_03, bone2Name=dyn_skirt_rear_r_02, swing1LimitDegrees=24, swing2LimitDegrees=14, twistLimitDegrees=14}
 {bone1Name=dyn_skirt_rear_r_04, bone2Name=dyn_skirt_rear_r_03, swing1LimitDegrees=24, swing2LimitDegrees=14, twistLimitDegrees=14}
 {bone1Name=dyn_skirt_rear_r_05, bone2Name=dyn_skirt_rear_r_04, swing1LimitDegrees=8, swing2LimitDegrees=8, twistLimitDegrees=4}
 {bone1Name=dyn_skirt_side_r_01, bone2Name=dyn_skirt_side_r_MSTR, swing1LimitDegrees=24, swing2LimitDegrees=14, twistLimitDegrees=14}
 {bone1Name=dyn_skirt_side_r_02, bone2Name=dyn_skirt_side_r_01, swing1LimitDegrees=24, swing2LimitDegrees=14, twistLimitDegrees=14}
 {bone1Name=dyn_skirt_side_r_03, bone2Name=dyn_skirt_side_r_02, swing1LimitDegrees=24, swing2LimitDegrees=14, twistLimitDegrees=14}
 {bone1Name=dyn_skirt_side_r_04, bone2Name=dyn_skirt_side_r_03, swing1LimitDegrees=24, swing2LimitDegrees=14, twistLimitDegrees=14}
 {bone1Name=dyn_skirt_side_r_05, bone2Name=dyn_skirt_side_r_04, swing1LimitDegrees=8, swing2LimitDegrees=8, twistLimitDegrees=4}
 {
  bone1Name=dyn_skirt_rear_r_03
  bone2Name=dyn_skirt_rear_l_03
  swing1LimitDegrees=45
  swing1Motion=Free
  swing2LimitDegrees=45
  swing2Motion=Free
  twistLimitDegrees=45
  twistMotion=Free
 }
 {
  bone1Name=dyn_skirt_rear_l_03
  bone2Name=dyn_skirt_rear_r_03
  swing1LimitDegrees=45
  swing1Motion=Free
  swing2LimitDegrees=45
  swing2Motion=Free
  twistLimitDegrees=45
  twistMotion=Free
 }
 {
  bone1Name=dyn_skirt_rear_r_04
  bone2Name=dyn_skirt_rear_l_04
  swing1LimitDegrees=45
  swing1Motion=Free
  swing2LimitDegrees=45
  swing2Motion=Free
  twistLimitDegrees=45
  twistMotion=Free
 }
 {
  bone1Name=dyn_skirt_rear_l_04
  bone2Name=dyn_skirt_rear_r_04
  swing1LimitDegrees=45
  swing1Motion=Free
  swing2LimitDegrees=45
  swing2Motion=Free
  twistLimitDegrees=45
  twistMotion=Free
 }
 {
  bone1Name=dyn_skirt_side_r_04
  bone2Name=dyn_skirt_rear_r_04
  swing1LimitDegrees=45
  swing1Motion=Free
  swing2LimitDegrees=45
  swing2Motion=Free
  twistLimitDegrees=45
  twistMotion=Free
 }
 {
  bone1Name=dyn_skirt_side_r_03
  bone2Name=dyn_skirt_rear_r_03
  swing1LimitDegrees=45
  swing1Motion=Free
  swing2LimitDegrees=45
  swing2Motion=Free
  twistLimitDegrees=45
  twistMotion=Free
 }
 {
  bone1Name=dyn_skirt_front_l_03
  bone2Name=dyn_skirt_side_l_03
  swing1LimitDegrees=45
  swing1Motion=Free
  swing2LimitDegrees=45
  swing2Motion=Free
  twistLimitDegrees=45
  twistMotion=Free
 }
 {
  bone1Name=dyn_skirt_front_l_04
  bone2Name=dyn_skirt_side_l_04
  swing1LimitDegrees=45
  swing1Motion=Free
  swing2LimitDegrees=45
  swing2Motion=Free
  twistLimitDegrees=45
  twistMotion=Free
 }
 {
  bone1Name=dyn_skirt_side_l_03
  bone2Name=dyn_skirt_rear_l_03
  swing1LimitDegrees=45
  swing1Motion=Free
  swing2LimitDegrees=45
  swing2Motion=Free
  twistLimitDegrees=45
  twistMotion=Free
 }
 {
  bone1Name=dyn_skirt_side_l_04
  bone2Name=dyn_skirt_rear_l_04
  swing1LimitDegrees=45
  swing1Motion=Free
  swing2LimitDegrees=45
  swing2Motion=Free
  twistLimitDegrees=45
  twistMotion=Free
 }
 {
  bone1Name=dyn_pouch_tie_l_MSTR
  bone2Name=pelvis
  swing1LimitDegrees=45
  swing1Motion=Free
  swing2LimitDegrees=45
  swing2Motion=Free
  twistLimitDegrees=45
  twistMotion=Free
 }
 {bone1Name=dyn_pouch_tie_l_01, bone2Name=dyn_pouch_tie_l_MSTR, swing1LimitDegrees=5, swing2LimitDegrees=5, twistLimitDegrees=5}
 {bone1Name=dyn_pouch_tie_l_02, bone2Name=dyn_pouch_tie_l_01, swing1LimitDegrees=20, swing2LimitDegrees=20, twistLimitDegrees=20}
 {
  bone1Name=dyn_skirt_front_l_01
  bone2Name=dyn_pouch_tie_l_01
  swing1LimitDegrees=8
  swing1Motion=Free
  swing2LimitDegrees=8
  swing2Motion=Free
  twistLimitDegrees=8
  twistMotion=Free
 }
 {
  bone1Name=dyn_pouch_tie_r_MSTR
  bone2Name=pelvis
  swing1LimitDegrees=45
  swing1Motion=Free
  swing2LimitDegrees=45
  swing2Motion=Free
  twistLimitDegrees=45
  twistMotion=Free
 }
 {bone1Name=dyn_pouch_tie_r_01, bone2Name=dyn_pouch_tie_r_MSTR, swing1LimitDegrees=5, swing2LimitDegrees=5, twistLimitDegrees=5}
 {bone1Name=dyn_pouch_tie_r_02, bone2Name=dyn_pouch_tie_r_01, swing1LimitDegrees=20, swing2LimitDegrees=20, twistLimitDegrees=20}
 {bone1Name=dyn_skirt_front_r_01, bone2Name=dyn_skirt_front_r_MSTR, swing1LimitDegrees=24, swing2LimitDegrees=14, twistLimitDegrees=14}
 {bone1Name=dyn_skirt_front_r_02, bone2Name=dyn_skirt_front_r_01, swing1LimitDegrees=24, swing2LimitDegrees=14, twistLimitDegrees=14}
 {bone1Name=dyn_skirt_front_r_03, bone2Name=dyn_skirt_front_r_02, swing1LimitDegrees=24, swing2LimitDegrees=14, twistLimitDegrees=14}
 {bone1Name=dyn_skirt_front_r_04, bone2Name=dyn_skirt_front_r_03, swing1LimitDegrees=24, swing2LimitDegrees=14, twistLimitDegrees=14}
 {bone1Name=dyn_skirt_front_r_05, bone2Name=dyn_skirt_front_r_04, swing1LimitDegrees=8, swing2LimitDegrees=8, twistLimitDegrees=4}
 {
  bone1Name=dyn_skirt_front_r_04
  bone2Name=dyn_skirt_side_r_04
  swing1LimitDegrees=45
  swing1Motion=Free
  swing2LimitDegrees=45
  swing2Motion=Free
  twistLimitDegrees=45
  twistMotion=Free
 }
 {
  bone1Name=dyn_skirt_front_r_03
  bone2Name=dyn_skirt_side_r_03
  swing1LimitDegrees=45
  swing1Motion=Free
  swing2LimitDegrees=45
  swing2Motion=Free
  twistLimitDegrees=45
  twistMotion=Free
 }
 {
  bone1Name=dyn_pouch_tie_r_02
  bone2Name=dyn_skirt_front_r_01
  swing1LimitDegrees=45
  swing1Motion=Locked
  swing2LimitDegrees=45
  swing2Motion=Locked
  twistLimitDegrees=45
  twistMotion=Locked
 }
]
