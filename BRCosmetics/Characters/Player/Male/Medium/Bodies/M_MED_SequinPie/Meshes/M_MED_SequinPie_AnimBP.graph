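# Post-process-style AnimBP for the SequinPie male cosmetic body, copying the parent mesh pose and adding secondary motion.
# A Control Rig feeds crouch-dependent floor-collision offsets and skirt and backpack bone adjustments, then AnimDynamics knots and a pelvis-space RigidBody simulation.
# Pouch-tie corrections finish the pose at full strength in front end, half otherwise.
# Events set movement-state gravity, front-end simulation settings and a matching-backbling flag.
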AnimBlueprint /BRCosmetics/Characters/Player/Male/Medium/Bodies/M_MED_SequinPie/Meshes/M_MED_SequinPie_AnimBP
target_skeleton=&/M_MED_SequinPie_Skeleton
graphs [
 AnimGraph {nodes [
   "Transform (Modify) Bone - Bone: dyn_pouch_tie_r_MSTR"(Alpha=1, Translation="0,0,0", Rotation="0,-3,0", Scale="1.000,0.814,1.000") {
    rotationMode=BMM_Additive
    rotationSpace=BCS_BoneSpace
    scaleMode=BMM_Additive
    scaleSpace=BCS_BoneSpace
    alphaBoolBlend {blendOption=Linear}
    boneToModify {boneName=dyn_pouch_tie_r_MSTR}
    rotation {yaw=-3}
    scale {y=0.814}
   }
   Output Pose(Result=Use cached pose 'Final Pose'()) {layerGroup=None}
   n1 = Control Rig(Source=Copy Pose From Mesh() {bUseAttachedParent=true}) {
    bTransferInputCurves=false
    alphaBoolBlend {blendOption=Linear}
    controlRigAssetReference {blueprintRigClass=&/M_MED_SequinPie_CtrlRig.M_MED_SequinPie_CtrlRig_C}
   }
   Save cached pose 'Post Deformation Rig'(Pose=@n1) {cacheName=Post Deformation Rig}
   n2 = "Transform (Modify) Bone - Bone: dyn_floor_collision"(
    ComponentPose=Local To Component(LocalPose=Use cached pose 'Post Deformation Rig'())
    Alpha=1
    Translation="0,0,-20"
    Rotation="0,0,0"
    Scale="1,1,1"
   ) {
    translationMode=BMM_Additive
    translationSpace=BCS_BoneSpace
    alphaBoolBlend {blendOption=Linear}
    boneToModify {boneName=dyn_floor_collision}
    translation {z=-20}
   }
   n3 = Component To Local(ComponentPose=@n2)
   n4 = "Transform (Modify) Bone - Bone: dyn_floor_collision"(
    ComponentPose=Local To Component(LocalPose=Use cached pose 'Post Deformation Rig'())
    Alpha=1
    Translation="0,0,-5"
    Rotation="0,0,0"
    Scale="1,1,1"
   ) {
    translationMode=BMM_Additive
    translationSpace=BCS_BoneSpace
    alphaBoolBlend {blendOption=Linear}
    boneToModify {boneName=dyn_floor_collision}
    translation {z=-5}
   }
   n5 = Component To Local(ComponentPose=@n4)
   n6 = Get bIsCrouching()
   n7 = "Transform (Modify) Bone - Bone: dyn_skirt_front_l_MSTR"!(
    ComponentPose=Local To Component(LocalPose=Blend Poses by bool(BlendPose_0=@n5, BlendPose_1=@n3, bActiveValue=@n6) {blendType=Linear})
    bAlphaBoolEnabled:bool=True
    Translation="0,0,0"
    Rotation="20,-70,0"
    Scale="1,1,1"
   ) -> (Pose) {
    alphaInputType=Bool
    rotationMode=BMM_Additive
    rotationSpace=BCS_BoneSpace
    translationMode=BMM_Additive
    translationSpace=BCS_BoneSpace
    alphaBoolBlend {blendInTime=0.1, blendOption=Linear, blendOutTime=0.1}
    boneToModify {boneName=dyn_skirt_front_l_MSTR}
    rotation {pitch=20, yaw=-70}
   }
   n8 = "Transform (Modify) Bone - Bone: dyn_skirt_rear_l_MSTR"!(ComponentPose=@n7, bAlphaBoolEnabled:bool=True, Translation="0,0,0", Rotation="0,-70,0", Scale="1,1,1") -> (Pose) {
    alphaInputType=Bool
    rotationMode=BMM_Additive
    rotationSpace=BCS_BoneSpace
    translationMode=BMM_Additive
    translationSpace=BCS_BoneSpace
    alphaBoolBlend {blendInTime=0.1, blendOption=Linear, blendOutTime=0.1}
    boneToModify {boneName=dyn_skirt_rear_l_MSTR}
    rotation {yaw=-70}
   }
   n9 = "Transform (Modify) Bone - Bone: dyn_skirt_side_l_MSTR"!(ComponentPose=@n8, bAlphaBoolEnabled:bool=True, Translation="0,0,0", Rotation="0,-60,0", Scale="1,1,1") -> (Pose) {
    alphaInputType=Bool
    rotationMode=BMM_Additive
    rotationSpace=BCS_BoneSpace
    translationMode=BMM_Additive
    translationSpace=BCS_BoneSpace
    alphaBoolBlend {blendInTime=0.1, blendOption=Linear, blendOutTime=0.1}
    boneToModify {boneName=dyn_skirt_side_l_MSTR}
    rotation {yaw=-60}
   }
   n10 = "Transform (Modify) Bone - Bone: dyn_skirt_front_r_MSTR"!(ComponentPose=@n9, bAlphaBoolEnabled:bool=True, Translation="0,0,0", Rotation="20,-80,0", Scale="1,1,1") -> (Pose) {
    alphaInputType=Bool
    rotationMode=BMM_Additive
    rotationSpace=BCS_BoneSpace
    translationMode=BMM_Additive
    translationSpace=BCS_BoneSpace
    alphaBoolBlend {blendInTime=0.1, blendOption=Linear, blendOutTime=0.1}
    boneToModify {boneName=dyn_skirt_front_r_MSTR}
    rotation {pitch=20, yaw=-80}
   }
   n11 = "Transform (Modify) Bone - Bone: dyn_skirt_rear_r_MSTR"!(ComponentPose=@n10, bAlphaBoolEnabled:bool=True, Translation="0,0,0", Rotation="0,-70,0", Scale="1,1,1") -> (Pose) {
    alphaInputType=Bool
    rotationMode=BMM_Additive
    rotationSpace=BCS_BoneSpace
    translationMode=BMM_Additive
    translationSpace=BCS_BoneSpace
    alphaBoolBlend {blendInTime=0.1, blendOption=Linear, blendOutTime=0.1}
    boneToModify {boneName=dyn_skirt_rear_r_MSTR}
    rotation {yaw=-70}
   }
   n12 = "Transform (Modify) Bone - Bone: dyn_skirt_side_r_MSTR"!(ComponentPose=@n11, bAlphaBoolEnabled:bool=True, Translation="0,0,0", Rotation="0,-80,0", Scale="1,1,1") -> (Pose) {
    alphaInputType=Bool
    rotationMode=BMM_Additive
    rotationSpace=BCS_BoneSpace
    translationMode=BMM_Additive
    translationSpace=BCS_BoneSpace
    alphaBoolBlend {blendInTime=0.1, blendOption=Linear, blendOutTime=0.1}
    boneToModify {boneName=dyn_skirt_side_r_MSTR}
    rotation {yaw=-80}
   }
   n13 = "Transform (Modify) Bone - Bone: attach_backpack"!(ComponentPose=@n12, bAlphaBoolEnabled:bool=True, Translation="0,0,2.4", Rotation="-10,0,5", Scale="1,1,1") -> (Pose) {
    alphaInputType=Bool
    rotationMode=BMM_Additive
    rotationSpace=BCS_BoneSpace
    translationMode=BMM_Additive
    translationSpace=BCS_BoneSpace
    alphaBoolBlend {blendOption=Linear}
    boneToModify {boneName=attach_backpack}
    rotation {pitch=-10, roll=5}
    translation {z=2.4}
   }
   Save cached pose 'Post Pose'(Pose=Component To Local(ComponentPose=@n13)) {cacheName=Post Pose}
   n14 = Anim Dynamics - Bone: dyn_elbow_knot_l_01!(
    SimulationSpace:AnimPhysSimSpaceType Enum=BoneRelative
    ComponentPose=Local To Component(LocalPose=Use cached pose 'Post Pose'())
    Alpha=1
    ExternalForce="0,0,0"
   ) -> (Pose) {
    angularDampingOverride=1
    angularSpringConstant=5000
    bAngularSpring=true
    bOverrideAngularDamping=true
    bUseGravityOverride=true
    gravityScale=0.35
    simulationSpace=BoneRelative
    alphaBoolBlend {blendOption=Linear}
    boundBone {boneName=dyn_elbow_knot_l_01}
    componentAppliedLinearAccClamp=<1000 1000 1000>
    componentLinearAccScale=<1 1 1>
    componentLinearVelScale=<1 1 1>
    gravityOverride {z=-980}
    physicsBodyDefinitions [{
      boundBone {boneName=dyn_elbow_knot_l_01}
      boxExtents=<10 20 20>
      constraintSetup {angularLimitsMax=<25 25 25>, angularLimitsMin=<-25 -25 -25>, angularTarget=<1 0 0>}
      localJointOffset=<4 0 0>
     }]
    relativeSpaceBone {boneName=lowerarm_l}
   }
   n15 = Anim Dynamics - Bone: dyn_elbow_knot_l_02!(SimulationSpace:AnimPhysSimSpaceType Enum=BoneRelative, ComponentPose=@n14, Alpha=1, ExternalForce="0,0,0") -> (Pose) {
    angularDampingOverride=1
    angularSpringConstant=5000
    bAngularSpring=true
    bOverrideAngularDamping=true
    bUseGravityOverride=true
    gravityScale=0.35
    simulationSpace=BoneRelative
    alphaBoolBlend {blendOption=Linear}
    boundBone {boneName=dyn_elbow_knot_l_02}
    componentAppliedLinearAccClamp=<1000 1000 1000>
    componentLinearAccScale=<1 1 1>
    componentLinearVelScale=<1 1 1>
    gravityOverride {z=-980}
    physicsBodyDefinitions [{
      boundBone {boneName=dyn_elbow_knot_l_02}
      boxExtents=<10 20 20>
      constraintSetup {angularLimitsMax=<25 25 25>, angularLimitsMin=<-25 -25 -25>, angularTarget=<1 0 0>}
      localJointOffset=<4 0 0>
     }]
    relativeSpaceBone {boneName=lowerarm_l}
   }
   n16 = Anim Dynamics - Bone: dyn_wrist_knot_l_01!(SimulationSpace:AnimPhysSimSpaceType Enum, ComponentPose=@n15, Alpha=1, ExternalForce="0,0,0") -> (Pose) {
    angularDampingOverride=1
    angularSpringConstant=5000
    bAngularSpring=true
    bOverrideAngularDamping=true
    bUseGravityOverride=true
    gravityScale=0.35
    simulationSpace=RootRelative
    alphaBoolBlend {blendOption=Linear}
    boundBone {boneName=dyn_wrist_knot_l_01}
    componentAppliedLinearAccClamp=<1000 1000 1000>
    componentLinearAccScale=<1 1 1>
    componentLinearVelScale=<1 1 1>
    gravityOverride {z=-980}
    physicsBodyDefinitions [{
      boundBone {boneName=dyn_wrist_knot_l_01}
      boxExtents=<10 20 20>
      constraintSetup {angularLimitsMax=<25 25 25>, angularLimitsMin=<-25 -25 -25>, angularTarget=<1 0 0>}
      localJointOffset=<4 0 0>
     }]
   }
   n17 = Anim Dynamics - Bone: dyn_wrist_knot_l_02!(SimulationSpace:AnimPhysSimSpaceType Enum, ComponentPose=@n16, Alpha=1, ExternalForce="0,0,0") -> (Pose) {
    angularDampingOverride=1
    angularSpringConstant=5000
    bAngularSpring=true
    bOverrideAngularDamping=true
    bUseGravityOverride=true
    gravityScale=0.35
    simulationSpace=RootRelative
    alphaBoolBlend {blendOption=Linear}
    boundBone {boneName=dyn_wrist_knot_l_02}
    componentAppliedLinearAccClamp=<1000 1000 1000>
    componentLinearAccScale=<1 1 1>
    componentLinearVelScale=<1 1 1>
    gravityOverride {z=-980}
    physicsBodyDefinitions [{
      boundBone {boneName=dyn_wrist_knot_l_02}
      boxExtents=<10 20 20>
      constraintSetup {angularLimitsMax=<25 25 25>, angularLimitsMin=<-25 -25 -25>, angularTarget=<1 0 0>}
      localJointOffset=<4 0 0>
     }]
   }
   n18 = Anim Dynamics - Bone: dyn_elbow_knot_r_02!(SimulationSpace:AnimPhysSimSpaceType Enum, ComponentPose=@n17, Alpha=1, ExternalForce="0,0,0") -> (Pose) {
    angularDampingOverride=1
    angularSpringConstant=5000
    bAngularSpring=true
    bOverrideAngularDamping=true
    bUseGravityOverride=true
    gravityScale=0.35
    simulationSpace=RootRelative
    alphaBoolBlend {blendOption=Linear}
    boundBone {boneName=dyn_elbow_knot_r_02}
    componentAppliedLinearAccClamp=<1000 1000 1000>
    componentLinearAccScale=<1 1 1>
    componentLinearVelScale=<1 1 1>
    gravityOverride {z=-980}
    physicsBodyDefinitions [{
      boundBone {boneName=dyn_elbow_knot_r_02}
      boxExtents=<10 20 20>
      constraintSetup {angularLimitsMax=<25 25 25>, angularLimitsMin=<-25 -25 -25>, angularTarget=<1 0 0>}
      localJointOffset=<-4 0 0>
     }]
   }
   n19 = Anim Dynamics - Bone: dyn_elbow_knot_r_01!(SimulationSpace:AnimPhysSimSpaceType Enum, ComponentPose=@n18, Alpha=1, ExternalForce="0,0,0") -> (Pose) {
    angularDampingOverride=1
    angularSpringConstant=5000
    bAngularSpring=true
    bOverrideAngularDamping=true
    bUseGravityOverride=true
    gravityScale=0.35
    simulationSpace=RootRelative
    alphaBoolBlend {blendOption=Linear}
    boundBone {boneName=dyn_elbow_knot_r_01}
    componentAppliedLinearAccClamp=<1000 1000 1000>
    componentLinearAccScale=<1 1 1>
    componentLinearVelScale=<1 1 1>
    gravityOverride {z=-980}
    physicsBodyDefinitions [{
      boundBone {boneName=dyn_elbow_knot_r_01}
      boxExtents=<10 20 20>
      constraintSetup {angularLimitsMax=<25 25 25>, angularLimitsMin=<-25 -25 -25>, angularTarget=<1 0 0>}
      localJointOffset=<-4 0 0>
     }]
   }
   n20 = Anim Dynamics - Bone: dyn_wrist_knot_r_02!(SimulationSpace:AnimPhysSimSpaceType Enum, ComponentPose=@n19, Alpha=1, ExternalForce="0,0,0") -> (Pose) {
    angularDampingOverride=1
    angularSpringConstant=5000
    bAngularSpring=true
    bOverrideAngularDamping=true
    bUseGravityOverride=true
    gravityScale=0.35
    simulationSpace=RootRelative
    alphaBoolBlend {blendOption=Linear}
    boundBone {boneName=dyn_wrist_knot_r_02}
    componentAppliedLinearAccClamp=<1000 1000 1000>
    componentLinearAccScale=<1 1 1>
    componentLinearVelScale=<1 1 1>
    gravityOverride {z=-980}
    physicsBodyDefinitions [{
      boundBone {boneName=dyn_wrist_knot_r_02}
      boxExtents=<10 20 20>
      constraintSetup {angularLimitsMax=<25 25 25>, angularLimitsMin=<-25 -25 -25>, angularTarget=<1 0 0>}
      localJointOffset=<-4 0 0>
     }]
   }
   n21 = Anim Dynamics - Bone: dyn_wrist_knot_r_01!(SimulationSpace:AnimPhysSimSpaceType Enum, ComponentPose=@n20, Alpha=1, ExternalForce="0,0,0") -> (Pose) {
    angularDampingOverride=1
    angularSpringConstant=5000
    bAngularSpring=true
    bOverrideAngularDamping=true
    bUseGravityOverride=true
    gravityScale=0.35
    simulationSpace=RootRelative
    alphaBoolBlend {blendOption=Linear}
    boundBone {boneName=dyn_wrist_knot_r_01}
    componentAppliedLinearAccClamp=<1000 1000 1000>
    componentLinearAccScale=<1 1 1>
    componentLinearVelScale=<1 1 1>
    gravityOverride {z=-980}
    physicsBodyDefinitions [{
      boundBone {boneName=dyn_wrist_knot_r_01}
      boxExtents=<10 20 20>
      constraintSetup {angularLimitsMax=<25 25 25>, angularLimitsMin=<-25 -25 -25>, angularTarget=<1 0 0>}
      localJointOffset=<-4 0 0>
     }]
   }
   n22 = Anim Dynamics - Bone: dyn_knee_knot_l_01!(SimulationSpace:AnimPhysSimSpaceType Enum, ComponentPose=@n21, Alpha=1, ExternalForce="0,0,0") -> (Pose) {
    angularDampingOverride=1
    angularSpringConstant=5000
    bAngularSpring=true
    bOverrideAngularDamping=true
    bUseGravityOverride=true
    gravityScale=0.35
    simulationSpace=RootRelative
    alphaBoolBlend {blendOption=Linear}
    boundBone {boneName=dyn_knee_knot_l_01}
    componentAppliedLinearAccClamp=<1000 1000 1000>
    componentLinearAccScale=<1 1 1>
    componentLinearVelScale=<1 1 1>
    gravityOverride {z=-980}
    physicsBodyDefinitions [{
      boundBone {boneName=dyn_knee_knot_l_01}
      boxExtents=<10 20 20>
      constraintSetup {angularLimitsMax=<30 30 30>, angularLimitsMin=<-30 -30 -30>, angularTarget=<1 0 0>}
      localJointOffset=<4 0 0>
     }]
   }
   n23 = Anim Dynamics - Bone: dyn_knee_knot_l_02!(SimulationSpace:AnimPhysSimSpaceType Enum, ComponentPose=@n22, Alpha=1, ExternalForce="0,0,0") -> (Pose) {
    angularDampingOverride=1
    angularSpringConstant=5000
    bAngularSpring=true
    bOverrideAngularDamping=true
    bUseGravityOverride=true
    gravityScale=0.35
    simulationSpace=RootRelative
    alphaBoolBlend {blendOption=Linear}
    boundBone {boneName=dyn_knee_knot_l_02}
    componentAppliedLinearAccClamp=<1000 1000 1000>
    componentLinearAccScale=<1 1 1>
    componentLinearVelScale=<1 1 1>
    gravityOverride {z=-980}
    physicsBodyDefinitions [{
      boundBone {boneName=dyn_knee_knot_l_02}
      boxExtents=<10 20 20>
      constraintSetup {angularLimitsMax=<30 30 30>, angularLimitsMin=<-30 -30 -30>, angularTarget=<1 0 0>}
      localJointOffset=<4 0 0>
     }]
   }
   n24 = Anim Dynamics - Bone: dyn_knee_knot_r_02!(SimulationSpace:AnimPhysSimSpaceType Enum, ComponentPose=@n23, Alpha=1, ExternalForce="0,0,0") -> (Pose) {
    angularDampingOverride=1
    angularSpringConstant=5000
    bAngularSpring=true
    bOverrideAngularDamping=true
    bUseGravityOverride=true
    gravityScale=0.35
    simulationSpace=RootRelative
    alphaBoolBlend {blendOption=Linear}
    boundBone {boneName=dyn_knee_knot_r_02}
    componentAppliedLinearAccClamp=<1000 1000 1000>
    componentLinearAccScale=<1 1 1>
    componentLinearVelScale=<1 1 1>
    gravityOverride {z=-980}
    physicsBodyDefinitions [{
      boundBone {boneName=dyn_knee_knot_r_02}
      boxExtents=<10 20 20>
      constraintSetup {angularLimitsMax=<30 30 30>, angularLimitsMin=<-30 -30 -30>, angularTarget=<1 0 0>}
      localJointOffset=<-4 0 0>
     }]
   }
   n25 = Anim Dynamics - Bone: dyn_knee_knot_r_01!(SimulationSpace:AnimPhysSimSpaceType Enum, ComponentPose=@n24, Alpha=1, ExternalForce="0,0,0") -> (Pose) {
    angularDampingOverride=1
    angularSpringConstant=5000
    bAngularSpring=true
    bOverrideAngularDamping=true
    bUseGravityOverride=true
    gravityScale=0.35
    simulationSpace=RootRelative
    alphaBoolBlend {blendOption=Linear}
    boundBone {boneName=dyn_knee_knot_r_01}
    componentAppliedLinearAccClamp=<1000 1000 1000>
    componentLinearAccScale=<1 1 1>
    componentLinearVelScale=<1 1 1>
    gravityOverride {z=-980}
    physicsBodyDefinitions [{
      boundBone {boneName=dyn_knee_knot_r_01}
      boxExtents=<10 20 20>
      constraintSetup {angularLimitsMax=<30 30 30>, angularLimitsMin=<-30 -30 -30>, angularTarget=<1 0 0>}
      localJointOffset=<-4 0 0>
     }]
   }
   n26 = RigidBody!(
    ComponentPose=@n25
    Alpha
    OverrideWorldGravity:Vector="0,0,-980"
    ExternalForce
    SimSpaceSettings:Sim Space Settings Structure=(WorldAlpha=0,VelocityScaleZ=1,DampingAlpha=1,MaxLinearVelocity=10000,MaxAngularVelocity=10000,MaxLinearAcceleration=10000,MaxAngularAcceleration=10000,ExternalLinearDragV=(X=0,Y=0,Z=0),ExternalLinearVelocity=(X=0,Y=0,Z=0),ExternalAngularVelocity=(X=0,Y=0,Z=0))
   ) -> (Pose) {
    bOverrideWorldGravity=true
    overridePhysicsAsset=&/M_MED_SequinPie_Physics
    simulationSpace=BaseBoneSpace
    alphaBoolBlend {blendOption=Linear}
    baseBoneRef {boneName=pelvis}
    componentAppliedLinearAccClamp=<2000 2000 2000>
    overrideWorldGravity {z=-980}
    simSpaceSettings {
     maxAngularAcceleration=40
     maxAngularVelocity=5
     maxLinearAcceleration=200
     maxLinearVelocity=200
     worldAlpha=1
     externalLinearDragV=<0.4 0.4 0.4>
    }
   }
   Save cached pose 'PostRBAN'(Pose=Component To Local(ComponentPose=@n26)) {cacheName=PostRBAN}
   n27 = "Transform (Modify) Bone - Bone: dyn_pouch_tie_r_MSTR"(
    ComponentPose=Local To Component(LocalPose=Use cached pose 'PostRBAN'())
    Alpha=1
    Translation="-0.099,0.000,-0.239"
    Rotation="7.942,36.512,8.245"
    Scale="1,1,1"
   ) {
    rotationMode=BMM_Additive
    rotationSpace=BCS_BoneSpace
    scaleMode=BMM_Additive
    scaleSpace=BCS_BoneSpace
    translationMode=BMM_Additive
    translationSpace=BCS_BoneSpace
    alphaBoolBlend {blendOption=Linear}
    boneToModify {boneName=dyn_pouch_tie_r_MSTR}
    rotation=<rot 7.942 36.512 8.245>
    translation {x=-0.099, z=-0.239}
   }
   n28 = "Transform (Modify) Bone - Bone: dyn_pouch_tie_r_02"(ComponentPose=@n27, Alpha=1, Translation="0.227,0.328,0.236", Rotation="-8.728,-29.336,4.885", Scale="1,1,1") {
    rotationMode=BMM_Additive
    rotationSpace=BCS_BoneSpace
    scaleMode=BMM_Additive
    scaleSpace=BCS_BoneSpace
    translationMode=BMM_Additive
    translationSpace=BCS_BoneSpace
    alphaBoolBlend {blendOption=Linear}
    boneToModify {boneName=dyn_pouch_tie_r_02}
    rotation=<rot -8.728 -29.336 4.885>
    translation=<0.227 0.328 0.236>
   }
   n29 = "Transform (Modify) Bone - Bone: dyn_pouch_tie_l_MSTR"(ComponentPose=@n28, Alpha=1, Translation="0,0,0", Rotation="-4.822,19.839,37.860", Scale="1,1,1") {
    rotationMode=BMM_Additive
    rotationSpace=BCS_BoneSpace
    scaleMode=BMM_Additive
    scaleSpace=BCS_BoneSpace
    translationMode=BMM_Additive
    translationSpace=BCS_BoneSpace
    alphaBoolBlend {blendOption=Linear}
    boneToModify {boneName=dyn_pouch_tie_l_MSTR}
    rotation=<rot -4.822 19.839 37.86>
   }
   n30 = "Transform (Modify) Bone - Bone: dyn_pouch_tie_l_02"(ComponentPose=@n29, Alpha=1, Translation="0,0,0", Rotation="0,-7,0", Scale="1,1,1") {
    rotationMode=BMM_Additive
    rotationSpace=BCS_BoneSpace
    scaleMode=BMM_Additive
    scaleSpace=BCS_BoneSpace
    translationMode=BMM_Additive
    translationSpace=BCS_BoneSpace
    alphaBoolBlend {blendOption=Linear}
    boneToModify {boneName=dyn_pouch_tie_l_02}
    rotation {yaw=-7}
   }
   n31 = "Transform (Modify) Bone - Bone: dyn_pouch_tie_r_MSTR"(
    ComponentPose=Local To Component(LocalPose=Use cached pose 'PostRBAN'())
    Alpha=0.5
    Translation="-0.099,0.000,-0.239"
    Rotation="7.942,36.512,8.245"
    Scale="1,1,1"
   ) {
    rotationMode=BMM_Additive
    rotationSpace=BCS_BoneSpace
    scaleMode=BMM_Additive
    scaleSpace=BCS_BoneSpace
    translationMode=BMM_Additive
    translationSpace=BCS_BoneSpace
    alphaBoolBlend {blendOption=Linear}
    boneToModify {boneName=dyn_pouch_tie_r_MSTR}
    rotation=<rot 7.942 36.512 8.245>
    translation {x=-0.099, z=-0.239}
   }
   n32 = "Transform (Modify) Bone - Bone: dyn_pouch_tie_r_02"(ComponentPose=@n31, Alpha=0.5, Translation="0.227,0.328,0.236", Rotation="-8.728,-29.336,4.885", Scale="1,1,1") {
    rotationMode=BMM_Additive
    rotationSpace=BCS_BoneSpace
    scaleMode=BMM_Additive
    scaleSpace=BCS_BoneSpace
    translationMode=BMM_Additive
    translationSpace=BCS_BoneSpace
    alphaBoolBlend {blendOption=Linear}
    boneToModify {boneName=dyn_pouch_tie_r_02}
    rotation=<rot -8.728 -29.336 4.885>
    translation=<0.227 0.328 0.236>
   }
   n33 = "Transform (Modify) Bone - Bone: dyn_pouch_tie_l_MSTR"(ComponentPose=@n32, Alpha=0.5, Translation="0,0,0", Rotation="-4.822,19.839,37.860", Scale="1,1,1") {
    rotationMode=BMM_Additive
    rotationSpace=BCS_BoneSpace
    scaleMode=BMM_Additive
    scaleSpace=BCS_BoneSpace
    translationMode=BMM_Additive
    translationSpace=BCS_BoneSpace
    alphaBoolBlend {blendOption=Linear}
    boneToModify {boneName=dyn_pouch_tie_l_MSTR}
    rotation=<rot -4.822 19.839 37.86>
   }
   n34 = "Transform (Modify) Bone - Bone: dyn_pouch_tie_l_02"(ComponentPose=@n33, Alpha=0.5, Translation="0,0,0", Rotation="0,-7,0", Scale="1,1,1") {
    rotationMode=BMM_Additive
    rotationSpace=BCS_BoneSpace
    scaleMode=BMM_Additive
    scaleSpace=BCS_BoneSpace
    translationMode=BMM_Additive
    translationSpace=BCS_BoneSpace
    alphaBoolBlend {blendOption=Linear}
    boneToModify {boneName=dyn_pouch_tie_l_02}
    rotation {yaw=-7}
   }
   Save cached pose 'Final Pose'(Pose=Blend Poses by bool(BlendPose_0=Component To Local(ComponentPose=@n34), BlendPose_1=Component To Local(ComponentPose=@n30), bActiveValue=Get bIsInFrontEnd()) {blendType=Linear}) {cacheName=Final Pose}
  ]}
 EventGraph {nodes [
   Event BlueprintUpdateAnimation()
   > n1 = Sequence()
   > n2 = Gravity Override State Function DEPRECATED(CharacterPartAnimInstance=Self-Reference(), GameTimeInSeconds=GetGameTimeInSeconds(), GravityOverrideParams=Get Gravity Override Params())
   > Set Gravity Override(Gravity Override=@n2)
   n3 = Branch(execute=@n1.then_1, Condition=Get bIsInFrontEnd())
   > Set Simulation Space(Simulation Space:AnimPhysSimSpaceType Enum=World) -> (Output_Get:AnimPhysSimSpaceType Enum)
   > Set IsNotFrontend(IsNotFrontend:bool=false) -> (Output_Get:bool)
   > Set Sim Space Settings(Sim Space Settings:Sim Space Settings Structure=Get Sim Space Settings_FE() -> (Sim Space Settings_FE:Sim Space Settings Structure)) -> (Output_Get:Sim Space Settings Structure)
   Set Simulation Space(execute=@n3.else)
   > Set IsNotFrontend(IsNotFrontend=true)
   > Set Sim Space Settings(Sim Space Settings=Get Sim Space Settings_Default() -> (Sim Space Settings_Default:Sim Space Settings Structure))
   Event OnCharacterCustomizationLoaded()
   > n4 = Branch(Condition=BackpackItemDefinitionHasAnyMetaTag!(Tags:Array of Gameplay Tag Structures=Make Array("[0]"=(TagName="Cosmetics.MetaTags.SequinPie")), bRequiresEquipped:bool=false) -> (ReturnValue:bool))
   > Set IsWearingSequinPieBackbling(IsWearingSequinPieBackbling:bool=true) -> (Output_Get:bool)
   Set IsWearingSequinPieBackbling(execute=@n4.else, IsWearingSequinPieBackbling=false)
   Event OnCharacterAndPickaxeCustomizationLoaded!() -> (
    OutputDelegate
    then
    CustomBaseMesh:Skeletal Mesh Component Object Reference
    CustomParts:Array of Custom Character Part Object References
    CustomMeshParts:Array of Skeletal Mesh Component Budgeted Object References
   ) {is=title K2Node_Event}
  ]}
]
variables [
 Gravity Override {type=Vector, default=<0 0 -980>}
 Gravity Override Params {
  type=GravityOverrideParamsStruct
  default {
   onGround_Moving {gravityOverride=<0 -400 -700>}
   skydive_FloatingDown {gravityOverride=<0 -700 400>, windAmplitude=<100 -100 150>, windFrequency=<1.5 7.5 8>}
   skydive_Parachuting {windAmplitude=<100 -300 200>}
   tactical_Sprint {gravityOverride=<0 -600 -500>}
  }
 }
 IsNotFrontend {type=boolean}
 IsWearingSequinPieBackbling {type=boolean}
 Sim Space Settings {
  type=SimSpaceSettings
  default {
   dampingAlpha=1
   maxAngularAcceleration=40
   maxAngularVelocity=5
   maxLinearAcceleration=200
   maxLinearVelocity=200
   velocityScaleZ=1
   worldAlpha=1
   externalLinearDragV=<0.4 0.4 0.4>
  }
 }
 Sim Space Settings_Default {
  type=SimSpaceSettings
  default {
   dampingAlpha=1
   maxAngularAcceleration=40
   maxAngularVelocity=5
   maxLinearAcceleration=200
   maxLinearVelocity=200
   velocityScaleZ=1
   worldAlpha=1
   externalLinearDragV=<0.4 0.4 0.4>
  }
 }
 Sim Space Settings_FE {
  type=SimSpaceSettings
  default {
   dampingAlpha=1
   maxAngularAcceleration=40
   maxAngularVelocity=5
   maxLinearAcceleration=100
   maxLinearVelocity=100
   velocityScaleZ=0
   worldAlpha=0.5
   externalLinearDragV=<0.4 0.4 0.4>
  }
 }
 Simulation Space {type=AnimPhysSimSpaceType}
]
