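# Physics asset of collision volumes for the cape on the male medium SequinPie player cosmetic body.
# Capsules cover pelvis, spine, clavicles, and the dynamic skirt and leg-armor bones; arm and leg bodies carry no shapes.
# A large offset sphere on dyn_floor_collision, which has no constraint, approximates the floor.
# Constraints link body parts and chain skirt and armor bones from spine_01, with all swing and twist motion free.
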PhysicsAsset /BRCosmetics/Characters/Player/Male/Medium/Bodies/M_MED_SequinPie/Meshes/Parts/M_MED_SequinPie_CapeColliders
bodies {
 calf_l []
 calf_r []
 clavicle_l [{shapeType=Capsule, length=17.0692, radius=11.3809, center=<7.3755 -1.0979 -1.955>, rotation=<rot 70.6387 -35.6443 141.3973>}]
 clavicle_r [{shapeType=Capsule, length=17.0692, radius=11.3809, center=<-7.3755 1.0979 1.955>, rotation=<rot 70.6387 -35.6443 -38.6027>}]
 dyn_floor_collision [{shapeType=Sphere, radius=300, shapeName=dyn_floor_collision_sphere, center=<0 -26.1642 -300>}]
 dyn_leg_armor_rear_l_03 [{
   shapeType=Capsule
   length=20.4375
   radius=8.2188
   shapeName=dyn_leg_armor_rear_l_03_capsule
   center=<4 3.1454 2.4921>
   rotation=<rot 0 -90 90>
  }]
 dyn_leg_armor_rear_l_05 [{
   shapeType=Capsule
   length=10.9429
   radius=5.028
   shapeName=dyn_leg_armor_rear_l_05_capsule
   center=<4.0264 0.441 0.9463>
   rotation=<rot 2.5299 -90 2.5445>
  }]
 dyn_leg_armor_rear_r_03 [{
   shapeType=Capsule
   length=20.4375
   radius=8.2188
   shapeName=dyn_leg_armor_rear_l_03_capsule
   center=<-4 -3.1454 -2.4921>
   rotation=<rot 0 -90 -90>
  }]
 dyn_leg_armor_rear_r_05 [{
   shapeType=Capsule
   length=10.9429
   radius=5.028
   shapeName=dyn_leg_armor_rear_l_05_capsule
   center=<-4.0264 -0.441 -0.9463>
   rotation=<rot 2.5299 -90 -177.4555>
  }]
 dyn_leg_armor_side_l_04 [{
   shapeType=Capsule
   length=7.4217
   radius=12.9291
   shapeName=dyn_leg_armor_side_l_04_capsule
   center=<7.9815 7.7881 -0.1685>
   rotation=<rot -3.6995 -90.0019 -88.7259>
  }]
 dyn_leg_armor_side_l_05 [{
   shapeType=Capsule
   length=13.1158
   radius=6.5818
   shapeName=dyn_leg_armor_side_l_05_capsule
   center=<7.7567 1.5325 -0.6914>
   rotation=<rot -4.9911 -90 1.6126>
  }]
 dyn_leg_armor_side_r_04 [{
   shapeType=Capsule
   length=7.4217
   radius=12.9291
   shapeName=dyn_leg_armor_side_l_04_capsule
   center=<-7.9815 -7.7881 0.1685>
   rotation=<rot -3.6995 -90.0019 91.2741>
  }]
 dyn_leg_armor_side_r_05 [{
   shapeType=Capsule
   length=13.1158
   radius=6.5818
   shapeName=dyn_leg_armor_side_l_05_capsule
   center=<-7.7567 -1.5325 0.6914>
   rotation=<rot -4.9911 -90 -178.3874>
  }]
 dyn_skirt_front_l_03 [{shapeType=Capsule, length=15.9425, radius=11.8175, shapeName=dyn_skirt_front_l_03_capsule, center=<0 4.4422 0>, rotation=<rot 0 -90 91.4674>}]
 dyn_skirt_front_l_05 [{shapeType=Capsule, length=15.9425, radius=11.8175, shapeName=dyn_skirt_front_l_05_capsule, center=<0 4.4422 0>, rotation=<rot 0 -90 91.4674>}]
 dyn_skirt_front_r_03 [{
   shapeType=Capsule
   length=15.9425
   radius=11.8175
   shapeName=dyn_skirt_front_l_03_capsule
   center=<0 -4.4422 0>
   rotation=<rot 0 -90 -88.5326>
  }]
 dyn_skirt_front_r_05 [{
   shapeType=Capsule
   length=15.9425
   radius=11.8175
   shapeName=dyn_skirt_front_l_05_capsule
   center=<0 -4.4422 0>
   rotation=<rot 0 -90 -88.5326>
  }]
 dyn_skirt_rear_l_03 [{shapeType=Capsule, length=15.3125, radius=14.6875, shapeName=dyn_skirt_rear_l_03_capsule, center=<0 4.6389 0>, rotation=<rot 90 0 0>}]
 dyn_skirt_rear_l_05 [{shapeType=Capsule, length=26.125, radius=14.6875, shapeName=dyn_skirt_rear_l_05_capsule, center=<0 4.6389 0>, rotation=<rot 90 0 0>}]
 dyn_skirt_rear_r_03 [{shapeType=Capsule, length=15.3125, radius=14.6875, shapeName=dyn_skirt_rear_l_03_capsule, center=<0 -4.6389 0>, rotation=<rot 90 0 180>}]
 dyn_skirt_rear_r_05 [{shapeType=Capsule, length=26.125, radius=14.6875, shapeName=dyn_skirt_rear_l_05_capsule, center=<0 -4.6389 0>, rotation=<rot 90 0 180>}]
 dyn_skirt_side_l_03 [{shapeType=Capsule, length=17.0625, radius=9.7238, shapeName=dyn_skirt_side_l_03_capsule, rotation=<rot 0 -90 -90.6951>}]
 dyn_skirt_side_l_05 [{shapeType=Capsule, length=17.0625, radius=9.7238, shapeName=dyn_skirt_side_l_05_capsule, rotation=<rot 0 -90 -90.6951>}]
 dyn_skirt_side_r_03 [{shapeType=Capsule, length=17.0625, radius=9.7238, shapeName=dyn_skirt_side_l_03_capsule, rotation=<rot 0 -90 89.3049>}]
 dyn_skirt_side_r_05 [{shapeType=Capsule, length=17.0625, radius=9.7238, shapeName=dyn_skirt_side_l_05_capsule, rotation=<rot 0 -90 89.3049>}]
 foot_l []
 foot_r []
 lowerarm_l []
 lowerarm_r []
 pelvis [{shapeType=Capsule, length=20.2456, radius=15.5388, center=<-5.416 0.1881 0>}]
 spine_01 [{shapeType=Capsule, length=7.9438, radius=13.5586, center=<-1.7965 0.1884 0>}]
 spine_02 [
  {shapeType=Capsule, length=7.3264, radius=13.6322, center=<-1.711 -0.1851 0.0001>}
  {shapeType=Capsule, length=29.3956, radius=11.5807, center=<2.5024 0.4992 3.9825>, rotation=<rot 90 0.0049 0.0049>}
  {shapeType=Capsule, length=29.875, radius=11.1013, center=<2.4704 0.4991 -3.9825>, rotation=<rot 90 -0.0396 -0.0396>}
 ]
 spine_03 [{shapeType=Capsule, length=8.2107, radius=13.7031, center=<0.1445 0.2396 0.0001>}]
 spine_04 [{shapeType=Capsule, length=9.4404, radius=14.1325, center=<2.766 0.7765 0.0002>}]
 spine_05 [
  {shapeType=Capsule, length=32.6476, radius=11.7575, center=<-0.0712 -0.3266 0>}
  {shapeType=Capsule, length=13.9518, radius=13.9425, center=<-0.2916 0.4452 0>}
 ]
 thigh_l []
 thigh_r []
 upperarm_l []
 upperarm_r []
}
constraints [
 {
  bone1Name=spine_01
  bone2Name=pelvis
  swing1LimitDegrees=45
  swing1Motion=Free
  swing2LimitDegrees=45
  swing2Motion=Free
  twistLimitDegrees=45
  twistMotion=Free
 }
 {
  bone1Name=spine_02
  bone2Name=spine_01
  swing1LimitDegrees=45
  swing1Motion=Free
  swing2LimitDegrees=45
  swing2Motion=Free
  twistLimitDegrees=45
  twistMotion=Free
 }
 {
  bone1Name=spine_03
  bone2Name=spine_02
  swing1LimitDegrees=45
  swing1Motion=Free
  swing2LimitDegrees=45
  swing2Motion=Free
  twistLimitDegrees=45
  twistMotion=Free
 }
 {
  bone1Name=spine_04
  bone2Name=spine_03
  swing1LimitDegrees=45
  swing1Motion=Free
  swing2LimitDegrees=45
  swing2Motion=Free
  twistLimitDegrees=45
  twistMotion=Free
 }
 {
  bone1Name=spine_05
  bone2Name=spine_04
  swing1LimitDegrees=45
  swing1Motion=Free
  swing2LimitDegrees=45
  swing2Motion=Free
  twistLimitDegrees=45
  twistMotion=Free
 }
 {
  bone1Name=thigh_l
  bone2Name=pelvis
  swing1LimitDegrees=45
  swing1Motion=Free
  swing2LimitDegrees=45
  swing2Motion=Free
  twistLimitDegrees=45
  twistMotion=Free
 }
 {
  bone1Name=foot_l
  bone2Name=calf_l
  swing1LimitDegrees=45
  swing1Motion=Free
  swing2LimitDegrees=45
  swing2Motion=Free
  twistLimitDegrees=45
  twistMotion=Free
 }
 {
  bone1Name=lowerarm_r
  bone2Name=upperarm_r
  swing1LimitDegrees=45
  swing1Motion=Free
  swing2LimitDegrees=45
  swing2Motion=Free
  twistLimitDegrees=45
  twistMotion=Free
 }
 {
  bone1Name=foot_r
  bone2Name=calf_r
  swing1LimitDegrees=45
  swing1Motion=Free
  swing2LimitDegrees=45
  swing2Motion=Free
  twistLimitDegrees=45
  twistMotion=Free
 }
 {
  bone1Name=lowerarm_l
  bone2Name=upperarm_l
  swing1LimitDegrees=45
  swing1Motion=Free
  swing2LimitDegrees=45
  swing2Motion=Free
  twistLimitDegrees=45
  twistMotion=Free
 }
 {
  bone1Name=upperarm_l
  bone2Name=clavicle_l
  swing1LimitDegrees=45
  swing1Motion=Free
  swing2LimitDegrees=45
  swing2Motion=Free
  twistLimitDegrees=45
  twistMotion=Free
 }
 {
  bone1Name=clavicle_l
  bone2Name=spine_05
  swing1LimitDegrees=45
  swing1Motion=Free
  swing2LimitDegrees=45
  swing2Motion=Free
  twistLimitDegrees=45
  twistMotion=Free
 }
 {
  bone1Name=calf_r
  bone2Name=thigh_r
  swing1LimitDegrees=45
  swing1Motion=Free
  swing2LimitDegrees=45
  swing2Motion=Free
  twistLimitDegrees=45
  twistMotion=Free
 }
 {
  bone1Name=thigh_r
  bone2Name=pelvis
  swing1LimitDegrees=45
  swing1Motion=Free
  swing2LimitDegrees=45
  swing2Motion=Free
  twistLimitDegrees=45
  twistMotion=Free
 }
 {
  bone1Name=calf_l
  bone2Name=thigh_l
  swing1LimitDegrees=45
  swing1Motion=Free
  swing2LimitDegrees=45
  swing2Motion=Free
  twistLimitDegrees=45
  twistMotion=Free
 }
 {
  bone1Name=dyn_skirt_rear_l_03
  bone2Name=spine_01
  swing1LimitDegrees=45
  swing1Motion=Free
  swing2LimitDegrees=45
  swing2Motion=Free
  twistLimitDegrees=45
  twistMotion=Free
 }
 {
  bone1Name=dyn_skirt_rear_l_05
  bone2Name=dyn_skirt_rear_l_03
  swing1LimitDegrees=45
  swing1Motion=Free
  swing2LimitDegrees=45
  swing2Motion=Free
  twistLimitDegrees=45
  twistMotion=Free
 }
 {
  bone1Name=dyn_leg_armor_rear_l_03
  bone2Name=spine_01
  swing1LimitDegrees=45
  swing1Motion=Free
  swing2LimitDegrees=45
  swing2Motion=Free
  twistLimitDegrees=45
  twistMotion=Free
 }
 {
  bone1Name=dyn_leg_armor_rear_r_03
  bone2Name=spine_01
  swing1LimitDegrees=45
  swing1Motion=Free
  swing2LimitDegrees=45
  swing2Motion=Free
  twistLimitDegrees=45
  twistMotion=Free
 }
 {
  bone1Name=dyn_skirt_rear_r_03
  bone2Name=spine_01
  swing1LimitDegrees=45
  swing1Motion=Free
  swing2LimitDegrees=45
  swing2Motion=Free
  twistLimitDegrees=45
  twistMotion=Free
 }
 {
  bone1Name=dyn_skirt_rear_r_05
  bone2Name=dyn_skirt_rear_r_03
  swing1LimitDegrees=45
  swing1Motion=Free
  swing2LimitDegrees=45
  swing2Motion=Free
  twistLimitDegrees=45
  twistMotion=Free
 }
 {
  bone1Name=upperarm_r
  bone2Name=clavicle_r
  swing1LimitDegrees=45
  swing1Motion=Free
  swing2LimitDegrees=45
  swing2Motion=Free
  twistLimitDegrees=45
  twistMotion=Free
 }
 {
  bone1Name=dyn_leg_armor_side_l_04
  bone2Name=spine_01
  swing1LimitDegrees=45
  swing1Motion=Free
  swing2LimitDegrees=45
  swing2Motion=Free
  twistLimitDegrees=45
  twistMotion=Free
 }
 {
  bone1Name=dyn_skirt_side_l_03
  bone2Name=spine_01
  swing1LimitDegrees=45
  swing1Motion=Free
  swing2LimitDegrees=45
  swing2Motion=Free
  twistLimitDegrees=45
  twistMotion=Free
 }
 {
  bone1Name=dyn_skirt_side_l_05
  bone2Name=dyn_skirt_side_l_03
  swing1LimitDegrees=45
  swing1Motion=Free
  swing2LimitDegrees=45
  swing2Motion=Free
  twistLimitDegrees=45
  twistMotion=Free
 }
 {
  bone1Name=dyn_skirt_side_r_03
  bone2Name=spine_01
  swing1LimitDegrees=45
  swing1Motion=Free
  swing2LimitDegrees=45
  swing2Motion=Free
  twistLimitDegrees=45
  twistMotion=Free
 }
 {
  bone1Name=dyn_skirt_side_r_05
  bone2Name=dyn_skirt_side_r_03
  swing1LimitDegrees=45
  swing1Motion=Free
  swing2LimitDegrees=45
  swing2Motion=Free
  twistLimitDegrees=45
  twistMotion=Free
 }
 {
  bone1Name=dyn_skirt_front_l_03
  bone2Name=spine_01
  swing1LimitDegrees=45
  swing1Motion=Free
  swing2LimitDegrees=45
  swing2Motion=Free
  twistLimitDegrees=45
  twistMotion=Free
 }
 {
  bone1Name=dyn_skirt_front_l_05
  bone2Name=dyn_skirt_front_l_03
  swing1LimitDegrees=45
  swing1Motion=Free
  swing2LimitDegrees=45
  swing2Motion=Free
  twistLimitDegrees=45
  twistMotion=Free
 }
 {
  bone1Name=dyn_leg_armor_side_r_04
  bone2Name=spine_01
  swing1LimitDegrees=45
  swing1Motion=Free
  swing2LimitDegrees=45
  swing2Motion=Free
  twistLimitDegrees=45
  twistMotion=Free
 }
 {
  bone1Name=dyn_skirt_front_r_03
  bone2Name=spine_01
  swing1LimitDegrees=45
  swing1Motion=Free
  swing2LimitDegrees=45
  swing2Motion=Free
  twistLimitDegrees=45
  twistMotion=Free
 }
 {
  bone1Name=dyn_skirt_front_r_05
  bone2Name=dyn_skirt_front_r_03
  swing1LimitDegrees=45
  swing1Motion=Free
  swing2LimitDegrees=45
  swing2Motion=Free
  twistLimitDegrees=45
  twistMotion=Free
 }
 {
  bone1Name=dyn_leg_armor_side_l_05
  bone2Name=dyn_leg_armor_side_l_04
  swing1LimitDegrees=45
  swing1Motion=Free
  swing2LimitDegrees=45
  swing2Motion=Free
  twistLimitDegrees=45
  twistMotion=Free
 }
 {
  bone1Name=dyn_leg_armor_side_r_05
  bone2Name=dyn_leg_armor_side_r_04
  swing1LimitDegrees=45
  swing1Motion=Free
  swing2LimitDegrees=45
  swing2Motion=Free
  twistLimitDegrees=45
  twistMotion=Free
 }
 {
  bone1Name=dyn_leg_armor_rear_l_05
  bone2Name=dyn_leg_armor_rear_l_03
  swing1LimitDegrees=45
  swing1Motion=Free
  swing2LimitDegrees=45
  swing2Motion=Free
  twistLimitDegrees=45
  twistMotion=Free
 }
 {
  bone1Name=dyn_leg_armor_rear_r_05
  bone2Name=dyn_leg_armor_rear_r_03
  swing1LimitDegrees=45
  swing1Motion=Free
  swing2LimitDegrees=45
  swing2Motion=Free
  twistLimitDegrees=45
  twistMotion=Free
 }
 {
  bone1Name=clavicle_r
  bone2Name=spine_05
  swing1LimitDegrees=45
  swing1Motion=Free
  swing2LimitDegrees=45
  swing2Motion=Free
  twistLimitDegrees=45
  twistMotion=Free
 }
]
