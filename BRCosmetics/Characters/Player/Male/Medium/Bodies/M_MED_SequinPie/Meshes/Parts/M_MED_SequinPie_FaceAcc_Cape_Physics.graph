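# Physics asset for the SequinPie male player cosmetic's cape, built from capsule bodies on the skeleton.
# Center, left and right six-bone chains hang from master bones on spine_05, with swing and twist limits generally loosening toward the hem.
# Free-rotation constraints cross-link the center chain to both sides at segments 03–05.
# Torso, clavicle, leg armor and rear skirt capsules, plus a large unconstrained floor capsule, appear to serve as cape colliders.
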PhysicsAsset /BRCosmetics/Characters/Player/Male/Medium/Bodies/M_MED_SequinPie/Meshes/Parts/M_MED_SequinPie_FaceAcc_Cape_Physics
bodies {
 clavicle_l [{
   shapeType=Capsule
   length=11.7824
   radius=9.9105
   shapeName=clavicle_l_capsule
   center=<12.0194 0.5385 -2.0987>
   rotation=<rot 90 -87.106 -90.7549>
  }]
 clavicle_r [{
   shapeType=Capsule
   length=11.7824
   radius=9.9105
   shapeName=clavicle_l_capsule
   center=<-12.0194 -0.5385 2.0987>
   rotation=<rot 90 -173.4634 2.8878>
  }]
 dyn_cape_01 [{shapeType=Capsule, length=11.2812, radius=8.7188, shapeName=dyn_cape_01_capsule, center=<4 0 6.5>, rotation=<rot 0 0 90>}]
 dyn_cape_02 [{shapeType=Capsule, length=13.5938, radius=6.4062, shapeName=dyn_cape_02_capsule, center=<4 0 3>, rotation=<rot 0 90 89.932>}]
 dyn_cape_03 [{shapeType=Capsule, length=12.5938, radius=7.4062, shapeName=dyn_cape_03_capsule, center=<4 0 3>, rotation=<rot 0 90 90>}]
 dyn_cape_04 [{shapeType=Capsule, length=11.1562, radius=8.8438, shapeName=dyn_cape_04_capsule, center=<4 0 5.3396>, rotation=<rot 0 90 90>}]
 dyn_cape_05 [{shapeType=Capsule, length=12.6562, radius=8.8438, shapeName=dyn_cape_05_capsule, center=<4 0 5.3396>, rotation=<rot 0 90 90>}]
 dyn_cape_06 [{shapeType=Capsule, length=17.7812, radius=11.9062, shapeName=dyn_cape_06_capsule, center=<4 0 6.3396>, rotation=<rot 0 90 90>}]
 dyn_cape_MSTR [{shapeType=Capsule, length=1, radius=1, shapeName=dyn_cape_MSTR_capsule, rotation=<rot 0 90 90>}]
 dyn_cape_l_01 [{shapeType=Capsule, length=12, radius=8, shapeName=dyn_cape_l_01_capsule, center=<4 0 6.5>, rotation=<rot -0.0001 0 83.0672>}]
 dyn_cape_l_02 [{shapeType=Capsule, length=13.5938, radius=6.4062, shapeName=dyn_cape_l_02_capsule, center=<4 0 3>, rotation=<rot 0.0001 90.0001 89.9321>}]
 dyn_cape_l_03 [{shapeType=Capsule, length=12.5938, radius=7.4062, shapeName=dyn_cape_l_03_capsule, center=<4 0 3>, rotation=<rot 0 90 90>}]
 dyn_cape_l_04 [{shapeType=Capsule, length=11.1562, radius=8.8438, shapeName=dyn_cape_l_04_capsule, center=<4 1 5.3396>, rotation=<rot 0 90 90>}]
 dyn_cape_l_05 [{shapeType=Capsule, length=12.6562, radius=8.8438, shapeName=dyn_cape_l_05_capsule, center=<4 1 5.3396>, rotation=<rot 0 90 90>}]
 dyn_cape_l_06 [{shapeType=Capsule, length=17.7812, radius=11.9062, shapeName=dyn_cape_l_06_capsule, center=<4 1 6.3396>, rotation=<rot 0 90 90>}]
 dyn_cape_l_MSTR [{shapeType=Capsule, length=1, radius=1, shapeName=dyn_cape_l_MSTR_capsule, rotation=<rot 0 90 90>}]
 dyn_cape_r_01 [{shapeType=Capsule, length=12, radius=8, shapeName=dyn_cape_l_01_capsule, center=<-4 0 -6.5>, rotation=<rot -0.0001 0 -96.9328>}]
 dyn_cape_r_02 [{shapeType=Capsule, length=13.5938, radius=6.4062, shapeName=dyn_cape_l_02_capsule, center=<-4 0 -3>, rotation=<rot 0 90 -90>}]
 dyn_cape_r_03 [{shapeType=Capsule, length=12.5938, radius=7.4062, shapeName=dyn_cape_l_03_capsule, center=<-4 0 -3>, rotation=<rot 0 90 -90>}]
 dyn_cape_r_04 [{shapeType=Capsule, length=11.1562, radius=8.8438, shapeName=dyn_cape_l_04_capsule, center=<-4 -1 -5.3396>, rotation=<rot 0 90 -90>}]
 dyn_cape_r_05 [{shapeType=Capsule, length=12.6562, radius=8.8438, shapeName=dyn_cape_l_05_capsule, center=<-4 -1 -5.3396>, rotation=<rot 0 90 -90>}]
 dyn_cape_r_06 [{shapeType=Capsule, length=17.7812, radius=11.9062, shapeName=dyn_cape_l_06_capsule, center=<-4 -1 -6.3396>, rotation=<rot 0 90 -90>}]
 dyn_cape_r_MSTR [{shapeType=Capsule, length=1, radius=1, shapeName=dyn_cape_r_MSTR_capsule, rotation=<rot 0 90 90>}]
 dyn_floor_collision [{shapeType=Capsule, length=0.505, radius=300, shapeName=dyn_floor_collision_capsule, center=<0 0 -300>, rotation=<rot 0 -90 0>}]
 dyn_leg_armor_rear_l_03 [{
   shapeType=Capsule
   length=20.4375
   radius=8.2188
   shapeName=dyn_leg_armor_rear_l_03_capsule
   center=<4 3.1454 2.4921>
   rotation=<rot 0 -90 90>
  }]
 dyn_leg_armor_rear_l_05 [{
   shapeType=Capsule
   length=9.0679
   radius=4.4342
   shapeName=dyn_leg_armor_rear_l_05_capsule
   center=<2.2297 0.4445 1.026>
   rotation=<rot 2.5299 -90 2.5445>
  }]
 dyn_leg_armor_rear_r_03 [{
   shapeType=Capsule
   length=20.4375
   radius=8.2188
   shapeName=dyn_leg_armor_rear_l_03_capsule
   center=<-4 -3.1454 -2.4921>
   rotation=<rot 0 -90 -90>
  }]
 dyn_leg_armor_rear_r_05 [{
   shapeType=Capsule
   length=9.0679
   radius=4.4342
   shapeName=dyn_leg_armor_rear_r_05_capsule
   center=<-2.2297 -0.4445 -1.026>
   rotation=<rot 2.5299 -90 2.5445>
  }]
 dyn_leg_armor_side_l_03 [{
   shapeType=Capsule
   length=16.1875
   radius=9.6875
   shapeName=dyn_leg_armor_side_l_03_capsule
   center=<14.2451 1.1708 -2.5196>
   rotation=<rot 81.1689 -49.2784 -49.6159>
  }]
 dyn_leg_armor_side_r_03 [{
   shapeType=Capsule
   length=13.9062
   radius=11.9688
   shapeName=dyn_leg_armor_side_l_03_capsule
   center=<-14.3445 -0.9092 0.2977>
   rotation=<rot 82.6664 -66.1925 113.6342>
  }]
 dyn_skirt_rear_l_03 [{shapeType=Capsule, length=15.3125, radius=14.6875, shapeName=dyn_skirt_rear_l_03_capsule, center=<0 4.6388 0>, rotation=<rot 90 0 0>}]
 dyn_skirt_rear_l_05 [{shapeType=Capsule, length=26.125, radius=14.6875, shapeName=dyn_skirt_rear_l_05_capsule, center=<0 4.6388 0>, rotation=<rot 90 0 0>}]
 dyn_skirt_rear_r_03 [{shapeType=Capsule, length=15.3125, radius=14.6875, shapeName=dyn_skirt_rear_l_03_capsule, center=<0 -4.6388 0>, rotation=<rot 90 0 180>}]
 dyn_skirt_rear_r_05 [{shapeType=Capsule, length=26.125, radius=14.6875, shapeName=dyn_skirt_rear_l_05_capsule, center=<0 -4.6388 0>, rotation=<rot 90 0 180>}]
 pelvis [{shapeType=Capsule, length=10, radius=10, shapeName=pelvis_capsule}]
 spine_02 [{shapeType=Capsule, length=10, radius=10, shapeName=spine_02_capsule}]
 spine_03 [{shapeType=Capsule, length=7.2812, radius=12.7188, shapeName=spine_03_capsule}]
 spine_04 [{shapeType=Capsule, length=6.5625, radius=13.4375, shapeName=spine_04_capsule, center=<1.0249 0.7182 0>}]
 spine_05 [{shapeType=Capsule, length=34.7281, radius=10.2581, shapeName=spine_05_capsule, center=<2.6803 -0.4226 0>}]
}
constraints [
 {bone1Name=dyn_cape_01, bone2Name=dyn_cape_MSTR, swing1LimitDegrees=4, swing2LimitDegrees=16, twistLimitDegrees=4}
 {bone1Name=dyn_cape_02, bone2Name=dyn_cape_01, swing1LimitDegrees=10, swing2LimitDegrees=16, twistLimitDegrees=10}
 {bone1Name=dyn_cape_03, bone2Name=dyn_cape_02, swing1LimitDegrees=10, swing2LimitDegrees=32, twistLimitDegrees=10}
 {bone1Name=dyn_cape_04, bone2Name=dyn_cape_03, swing1LimitDegrees=10, swing2LimitDegrees=32, twistLimitDegrees=10}
 {bone1Name=dyn_cape_05, bone2Name=dyn_cape_04, swing1LimitDegrees=18, swing2LimitDegrees=28, twistLimitDegrees=18}
 {bone1Name=dyn_cape_06, bone2Name=dyn_cape_05, swing1LimitDegrees=18, swing2LimitDegrees=38, twistLimitDegrees=18}
 {bone1Name=dyn_cape_l_02, bone2Name=dyn_cape_l_01, swing1LimitDegrees=10, swing2LimitDegrees=16, twistLimitDegrees=10}
 {bone1Name=dyn_cape_l_03, bone2Name=dyn_cape_l_02, swing1LimitDegrees=10, swing2LimitDegrees=32, twistLimitDegrees=10}
 {bone1Name=dyn_cape_l_04, bone2Name=dyn_cape_l_03, swing1LimitDegrees=10, swing2LimitDegrees=32, twistLimitDegrees=10}
 {bone1Name=dyn_cape_l_05, bone2Name=dyn_cape_l_04, swing1LimitDegrees=18, swing2LimitDegrees=28, twistLimitDegrees=18}
 {bone1Name=dyn_cape_l_06, bone2Name=dyn_cape_l_05, swing1LimitDegrees=18, swing2LimitDegrees=38, twistLimitDegrees=18}
 {
  bone1Name=dyn_cape_MSTR
  bone2Name=spine_05
  swing1LimitDegrees=45
  swing1Motion=Free
  swing2LimitDegrees=45
  swing2Motion=Free
  twistLimitDegrees=45
  twistMotion=Free
 }
 {
  bone1Name=spine_04
  bone2Name=spine_03
  swing1LimitDegrees=45
  swing1Motion=Free
  swing2LimitDegrees=45
  swing2Motion=Free
  twistLimitDegrees=45
  twistMotion=Free
 }
 {
  bone1Name=spine_03
  bone2Name=spine_02
  swing1LimitDegrees=45
  swing1Motion=Free
  swing2LimitDegrees=45
  swing2Motion=Free
  twistLimitDegrees=45
  twistMotion=Free
 }
 {
  bone1Name=spine_02
  bone2Name=pelvis
  swing1LimitDegrees=45
  swing1Motion=Free
  swing2LimitDegrees=45
  swing2Motion=Free
  twistLimitDegrees=45
  twistMotion=Free
 }
 {bone1Name=dyn_cape_r_01, bone2Name=dyn_cape_r_MSTR, swing1LimitDegrees=4, swing2LimitDegrees=16, twistLimitDegrees=4}
 {bone1Name=dyn_cape_r_02, bone2Name=dyn_cape_r_01, swing1LimitDegrees=10, swing2LimitDegrees=16, twistLimitDegrees=10}
 {bone1Name=dyn_cape_r_03, bone2Name=dyn_cape_r_02, swing1LimitDegrees=10, swing2LimitDegrees=32, twistLimitDegrees=10}
 {bone1Name=dyn_cape_r_04, bone2Name=dyn_cape_r_03, swing1LimitDegrees=10, swing2LimitDegrees=32, twistLimitDegrees=10}
 {bone1Name=dyn_cape_r_05, bone2Name=dyn_cape_r_04, swing1LimitDegrees=18, swing2LimitDegrees=28, twistLimitDegrees=18}
 {bone1Name=dyn_cape_r_06, bone2Name=dyn_cape_r_05, swing1LimitDegrees=18, swing2LimitDegrees=38, twistLimitDegrees=18}
 {
  bone1Name=spine_05
  bone2Name=spine_04
  swing1LimitDegrees=45
  swing1Motion=Free
  swing2LimitDegrees=45
  swing2Motion=Free
  twistLimitDegrees=45
  twistMotion=Free
 }
 {
  bone1Name=dyn_leg_armor_rear_l_03
  bone2Name=pelvis
  swing1LimitDegrees=45
  swing1Motion=Free
  swing2LimitDegrees=45
  swing2Motion=Free
  twistLimitDegrees=45
  twistMotion=Free
 }
 {
  bone1Name=dyn_leg_armor_rear_l_05
  bone2Name=dyn_leg_armor_rear_l_03
  swing1LimitDegrees=45
  swing1Motion=Free
  swing2LimitDegrees=45
  swing2Motion=Free
  twistLimitDegrees=45
  twistMotion=Free
 }
 {
  bone1Name=dyn_leg_armor_rear_r_03
  bone2Name=pelvis
  swing1LimitDegrees=45
  swing1Motion=Free
  swing2LimitDegrees=45
  swing2Motion=Free
  twistLimitDegrees=45
  twistMotion=Free
 }
 {
  bone1Name=dyn_leg_armor_rear_r_05
  bone2Name=dyn_leg_armor_rear_r_03
  swing1LimitDegrees=45
  swing1Motion=Free
  swing2LimitDegrees=45
  swing2Motion=Free
  twistLimitDegrees=45
  twistMotion=Free
 }
 {
  bone1Name=dyn_skirt_rear_l_03
  bone2Name=pelvis
  swing1LimitDegrees=45
  swing1Motion=Free
  swing2LimitDegrees=45
  swing2Motion=Free
  twistLimitDegrees=45
  twistMotion=Free
 }
 {
  bone1Name=dyn_skirt_rear_l_05
  bone2Name=dyn_skirt_rear_l_03
  swing1LimitDegrees=45
  swing1Motion=Free
  swing2LimitDegrees=45
  swing2Motion=Free
  twistLimitDegrees=45
  twistMotion=Free
 }
 {
  bone1Name=dyn_skirt_rear_r_03
  bone2Name=pelvis
  swing1LimitDegrees=45
  swing1Motion=Free
  swing2LimitDegrees=45
  swing2Motion=Free
  twistLimitDegrees=45
  twistMotion=Free
 }
 {
  bone1Name=dyn_skirt_rear_r_05
  bone2Name=dyn_skirt_rear_r_03
  swing1LimitDegrees=45
  swing1Motion=Free
  swing2LimitDegrees=45
  swing2Motion=Free
  twistLimitDegrees=45
  twistMotion=Free
 }
 {
  bone1Name=dyn_cape_05
  bone2Name=dyn_cape_l_05
  swing1LimitDegrees=10
  swing1Motion=Free
  swing2LimitDegrees=42
  swing2Motion=Free
  twistLimitDegrees=10
  twistMotion=Free
 }
 {
  bone1Name=dyn_cape_05
  bone2Name=dyn_cape_r_05
  swing1LimitDegrees=10
  swing1Motion=Free
  swing2LimitDegrees=42
  swing2Motion=Free
  twistLimitDegrees=10
  twistMotion=Free
 }
 {
  bone1Name=dyn_cape_04
  bone2Name=dyn_cape_l_04
  swing1LimitDegrees=10
  swing1Motion=Free
  swing2LimitDegrees=42
  swing2Motion=Free
  twistLimitDegrees=10
  twistMotion=Free
 }
 {
  bone1Name=dyn_cape_04
  bone2Name=dyn_cape_r_04
  swing1LimitDegrees=10
  swing1Motion=Free
  swing2LimitDegrees=42
  swing2Motion=Free
  twistLimitDegrees=10
  twistMotion=Free
 }
 {
  bone1Name=dyn_cape_03
  bone2Name=dyn_cape_l_03
  swing1LimitDegrees=20
  swing1Motion=Free
  swing2LimitDegrees=20
  swing2Motion=Free
  twistLimitDegrees=20
  twistMotion=Free
 }
 {
  bone1Name=dyn_cape_03
  bone2Name=dyn_cape_r_03
  swing1LimitDegrees=20
  swing1Motion=Free
  swing2LimitDegrees=20
  swing2Motion=Free
  twistLimitDegrees=20
  twistMotion=Free
 }
 {
  bone1Name=dyn_leg_armor_side_l_03
  bone2Name=pelvis
  swing1LimitDegrees=45
  swing1Motion=Free
  swing2LimitDegrees=45
  swing2Motion=Free
  twistLimitDegrees=45
  twistMotion=Free
 }
 {
  bone1Name=dyn_leg_armor_side_r_03
  bone2Name=pelvis
  swing1LimitDegrees=45
  swing1Motion=Free
  swing2LimitDegrees=45
  swing2Motion=Free
  twistLimitDegrees=45
  twistMotion=Free
 }
 {
  bone1Name=dyn_cape_l_MSTR
  bone2Name=spine_05
  swing1LimitDegrees=45
  swing1Motion=Free
  swing2LimitDegrees=45
  swing2Motion=Free
  twistLimitDegrees=45
  twistMotion=Free
 }
 {
  bone1Name=dyn_cape_r_MSTR
  bone2Name=spine_05
  swing1LimitDegrees=45
  swing1Motion=Free
  swing2LimitDegrees=45
  swing2Motion=Free
  twistLimitDegrees=45
  twistMotion=Free
 }
 {
  bone1Name=clavicle_l
  bone2Name=spine_05
  swing1LimitDegrees=45
  swing1Motion=Free
  swing2LimitDegrees=45
  swing2Motion=Free
  twistLimitDegrees=45
  twistMotion=Free
 }
 {bone1Name=dyn_cape_l_01, bone2Name=dyn_cape_l_MSTR, swing1LimitDegrees=10, swing2LimitDegrees=16, twistLimitDegrees=10}
 {
  bone1Name=clavicle_r
  bone2Name=spine_05
  swing1LimitDegrees=45
  swing1Motion=Free
  swing2LimitDegrees=45
  swing2Motion=Free
  twistLimitDegrees=45
  twistMotion=Free
 }
]
