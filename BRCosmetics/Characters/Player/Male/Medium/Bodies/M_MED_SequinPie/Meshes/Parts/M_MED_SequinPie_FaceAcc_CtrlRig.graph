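# Control Rig for the M_MED_SequinPie face accessory part that procedurally drives leg armor, cape, cape tie and shoulder armor bones from the body pose.
# Four functions run in sequence, using parent and aim constraints, spherical pose readers that weight corrective transforms, and swing-twist extraction.
# Leg armor follows skirt bones and a spawned front control blended between thighs.
# Cape aiming branches on Anim Dynamics Cape and WearingBackbling variables.
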ControlRigBlueprint /BRCosmetics/Characters/Player/Male/Medium/Bodies/M_MED_SequinPie/Meshes/Parts/M_MED_SequinPie_FaceAcc_CtrlRig
graphs [
 ConstructionGraph {nodes [
   RigUnit_PrepareForExecution = &RigUnit_PrepareForExecution()
   RigUnit_HierarchyAddControlTransform = /Script/ControlRig.RigUnit_HierarchyAddControlTransform(
    InitialValue=(Rotation=(X=0,Y=0,Z=0,W=1),Translation=(X=0,Y=0,Z=0),Scale3D=(X=1,Y=1,Z=1))
    Name=dyn_leg_armor_front_05_ctrl
    OffsetSpace=LocalSpace
    OffsetTransform=(Rotation=(X=0,Y=0,Z=0.9983,W=0.0587),Translation=(X=-15.0704,Y=12.7942,Z=0),Scale3D=(X=1,Y=1,Z=1))
    Parent=(Type=Bone,Name="pelvis")
    Settings="(InitialSpace=LocalSpace,bUsePreferredRotationOrder=False,PreferredRotationOrder=YZX,Limits=(LimitTranslationX=(bMinimum=False,bMaximum=False),LimitTranslationY=(bMinimum=False,bMaximum=False),LimitTranslationZ=(bMinimum=False,bMaximum=False),LimitPitch=(b ...(truncated)"
    ExecutePin=@RigUnit_PrepareForExecution.ExecutePin
   ) {
    is=struct RigVMUnitNode
    template="SpawnControl::Execute(in InitialValue,in Settings,in OffsetTransform,in OffsetSpace,in Parent,in Name,out Item)"
   }
  ]}
 New Function_ContainedGraph {nodes [
   RigVMFunctionEntryNode()
   > RigVMFunction_Sequence = RigVMAggregateNode()
   &RigUnit_ParentConstraint(
    AdvancedSettings=(InterpolationType=Average,RotationOrderForFilter=XZY)
    Child=(Type=Bone,Name="dyn_leg_armor_side_l_MSTR")
    Filter=(TranslationFilter=(bX=True,bY=True,bZ=True),RotationFilter=(bX=true,bY=True,bZ=True),ScaleFilter=(bX=True,bY=True,bZ=True))
    Parents=((Item=(Type=Bone,Name="dyn_skirt_front_l_01"),Weight=1))
    Weight=1
    ExecutePin=@RigVMFunction_Sequence.B
   )
   > &RigUnit_ParentConstraint(
    AdvancedSettings=(InterpolationType=Average,RotationOrderForFilter=XZY)
    Child=(Type=Bone,Name="dyn_leg_armor_rear_l_MSTR")
    Filter=(TranslationFilter=(bX=True,bY=True,bZ=True),RotationFilter=(bX=true,bY=True,bZ=True),ScaleFilter=(bX=True,bY=True,bZ=True))
    Parents=((Item=(Type=Bone,Name="dyn_skirt_rear_l_01"),Weight=1))
    Weight=1
   )
   > &RigUnit_ParentConstraint(
    AdvancedSettings=(InterpolationType=Average,RotationOrderForFilter=XZY)
    Child=(Type=Bone,Name="dyn_leg_armor_side_r_MSTR")
    Filter=(TranslationFilter=(bX=True,bY=True,bZ=True),RotationFilter=(bX=true,bY=true,bZ=True),ScaleFilter=(bX=True,bY=True,bZ=True))
    Parents=((Item=(Type=Bone,Name="dyn_skirt_front_r_01"),Weight=1))
    Weight=1
   )
   > &RigUnit_ParentConstraint(
    AdvancedSettings=(InterpolationType=Average,RotationOrderForFilter=XZY)
    Child=(Type=Bone,Name="dyn_leg_armor_rear_r_MSTR")
    Filter=(TranslationFilter=(bX=True,bY=True,bZ=True),RotationFilter=(bX=true,bY=true,bZ=True),ScaleFilter=(bX=True,bY=True,bZ=True))
    Parents=((Item=(Type=Bone,Name="dyn_skirt_rear_r_01"),Weight=1))
    Weight=1
   )
   > &RigUnit_AimConstraintLocalSpaceOffset(
    Child=(Type=Bone,Name="dyn_leg_armor_side_l_MSTR")
    Filter=(bX=false,bY=True,bZ=True)
    Parents=((Item=(Type=Bone,Name="dyn_skirt_front_l_02"),Weight=1))
    Weight=1
    WorldUp=(Target=(X=0,Y=0,Z=1),Kind=Location,Space=(Type=Bone,Name="dyn_skirt_front_l_02"))
   )
   > &RigUnit_AimConstraintLocalSpaceOffset(
    Child=(Type=Bone,Name="dyn_leg_armor_side_r_MSTR")
    Filter=(bX=false,bY=True,bZ=True)
    Parents=((Item=(Type=Bone,Name="dyn_skirt_front_r_02"),Weight=1))
    Weight=1
    WorldUp=(Target=(X=0,Y=0,Z=1),Kind=Location,Space=(Type=Bone,Name="dyn_skirt_front_r_02"))
   )
   > &RigUnit_ModifyTransforms(
    ItemToModify="((Item=(Type=Bone,Name=\"dyn_leg_armor_side_l_MSTR\"),Transform=(Rotation=(X=0.0317,Y=-0.0014,Z=-0.0436,W=0.9985),Translation=(X=1,Y=-2,Z=0),Scale3D=(X=1,Y=1,Z=1))),(Item=(Type=Bone,Name=\"dyn_leg_armor_rear_l_MSTR\"),Transform=(Rotation=(X=0,Y=0,Z=-0.0654,W=0 ...(truncated)"
    Weight=0.5
    WeightMinimum=0
   )
   RigVMFunctionReturnNode(ExecuteContext=@RigVMFunction_Sequence.D)
   n1 = &RigUnit_SphericalPoseReader(
    ActiveRegionScaleFactors=(PositiveWidth=3.5,NegativeWidth=0.2,PositiveHeight=3,NegativeHeight=3)
    ActiveRegionSize=0.1
    Debug=(bDrawDebug=false,DebugScale=25,DebugSegments=20,DebugThickness=0.25)
    DriverAxis=(X=1,Y=0,Z=0)
    DriverItem=(Type=Bone,Name="thigh_r")
    FalloffRegionScaleFactors=(PositiveWidth=1,NegativeWidth=1,PositiveHeight=1,NegativeHeight=1)
    FalloffSize=0.59
    FlipHeightScaling=false
    FlipWidthScaling=true
    OptionalParentItem=(Type=Bone)
    RotationOffset=(X=-90,Y=0,Z=-15)
    ExecutePin=@RigVMFunction_Sequence.A
   )
   > n2 = &RigUnit_SphericalPoseReader(
    ActiveRegionScaleFactors=(PositiveWidth=3.5,NegativeWidth=0.5,PositiveHeight=3,NegativeHeight=3)
    ActiveRegionSize=0.1
    Debug=(bDrawDebug=false,DebugScale=25,DebugSegments=20,DebugThickness=0.25)
    DriverAxis=(X=-1,Y=0,Z=0)
    DriverItem=(Type=Bone,Name="thigh_l")
    FalloffRegionScaleFactors=(PositiveWidth=1,NegativeWidth=1,PositiveHeight=1,NegativeHeight=1)
    FalloffSize=0.59
    OptionalParentItem=(Type=Bone)
    RotationOffset=(X=90,Y=0,Z=-15)
   )
   > /Script/ControlRig.RigUnit_PositionConstraintLocalSpaceOffset(
    Child=(Type=Control,Name="dyn_leg_armor_front_05_ctrl")
    Filter=(bX=true,bY=True,bZ=true)
    Parents=((Item=(Type=Bone,Name="thigh_twist_01_l"),Weight=1),(Item=(Type=Bone,Name="thigh_twist_01_r"),Weight=1))
    Weight=1
    bMaintainOffset=true
    Parents.0.Weight=@n2.OutputParam
    Parents.1.Weight=@n1.OutputParam
   ) {is=struct RigVMUnitNode}
   > &RigUnit_AimConstraintLocalSpaceOffset(
    Child=(Type=Bone,Name="dyn_leg_armor_front_MSTR")
    Filter=(bX=True,bY=True,bZ=True)
    Parents=((Item=(Type=Control,Name="dyn_leg_armor_front_05_ctrl"),Weight=1))
    Weight=1
    WorldUp=(Target=(X=0,Y=0,Z=1),Kind=Location,Space=(Type=Control,Name="dyn_leg_armor_front_05_ctrl"))
   )
   > Get Delta Ref = RigVMFunctionReferenceNode(Item=(Type=Bone,Name="dyn_leg_armor_front_MSTR"))
   Remap = &RigVMFunction_MathDoubleRemap(SourceMaximum=-45, SourceMinimum=0, TargetMaximum=1, TargetMinimum=0, Value=@"Get Delta Ref".Result.Rotation.Yaw)
   > &RigUnit_ModifyTransforms(
    ItemToModify=((Item=(Type=Bone,Name="dyn_leg_armor_front_MSTR"),Transform=(Rotation=(X=0,Y=0,Z=0,W=1),Translation=(X=2,Y=0,Z=0),Scale3D=(X=1,Y=1,Z=1))))
    WeightMinimum=0
    Weight=@Remap.Result
   )
   n3 = &RigUnit_SphericalPoseReader(
    ActiveRegionScaleFactors=(PositiveWidth=3.5,NegativeWidth=1,PositiveHeight=3.5,NegativeHeight=3.5)
    ActiveRegionSize=0.1
    Debug=(bDrawDebug=true,bDraw2D=False,bDrawLocalAxes=False,DebugScale=25,DebugSegments=20,DebugThickness=0.25)
    DriverAxis=(X=-1,Y=0,Z=0)
    DriverItem=(Type=Bone,Name="dyn_skirt_front_r_02")
    FalloffRegionScaleFactors=(PositiveWidth=1,NegativeWidth=1,PositiveHeight=1,NegativeHeight=1)
    FalloffSize=0.4
    FlipHeightScaling=False
    FlipWidthScaling=False
    OptionalParentItem=(Type=Bone,Name="dyn_skirt_front_r_MSTR")
    RotationOffset=(X=90,Y=0,Z=0)
    ExecutePin=@RigVMFunction_Sequence.C
   )
   > &RigUnit_ModifyTransforms(
    ItemToModify="((Item=(Type=Bone,Name=\"dyn_leg_armor_side_r_03\"),Transform=(Rotation=(X=0,Y=0,Z=-0.2481,W=0.9687),Translation=(X=0,Y=0,Z=0),Scale3D=(X=1,Y=1,Z=1))),(Item=(Type=Bone,Name=\"dyn_leg_armor_side_r_04\"),Transform=(Rotation=(X=0,Y=0,Z=-0.3403,W=0.9403),Translati ...(truncated)"
    WeightMinimum=0
    Weight=@n3.OutputParam
   )
   > n4 = &RigUnit_SphericalPoseReader(
    ActiveRegionScaleFactors=(PositiveWidth=3.5,NegativeWidth=1,PositiveHeight=3.5,NegativeHeight=3.5)
    ActiveRegionSize=0.1
    Debug=(bDrawDebug=true,bDraw2D=False,bDrawLocalAxes=False,DebugScale=25,DebugSegments=20,DebugThickness=0.25)
    DriverAxis=(X=1,Y=0,Z=0)
    DriverItem=(Type=Bone,Name="dyn_skirt_front_l_02")
    FalloffRegionScaleFactors=(PositiveWidth=1,NegativeWidth=1,PositiveHeight=1,NegativeHeight=1)
    FalloffSize=0.4
    FlipHeightScaling=false
    FlipWidthScaling=true
    OptionalParentItem=(Type=Bone,Name="dyn_skirt_front_l_MSTR")
    RotationOffset=(X=-90,Y=0,Z=0)
   )
   > &RigUnit_ModifyTransforms(
    ItemToModify="((Item=(Type=Bone,Name=\"dyn_leg_armor_side_l_03\"),Transform=(Rotation=(X=0,Y=0,Z=-0.2481,W=0.9687),Translation=(X=0,Y=0,Z=0),Scale3D=(X=1,Y=1,Z=1))),(Item=(Type=Bone,Name=\"dyn_leg_armor_side_l_04\"),Transform=(Rotation=(X=0,Y=0,Z=-0.3403,W=0.9403),Translati ...(truncated)"
    WeightMinimum=0
    Weight=@n4.OutputParam
   )
   &RigUnit_AimConstraintLocalSpaceOffset(
    Child=(Type=Bone,Name="dyn_leg_armor_rear_l_MSTR")
    ExecutePin=()
    Filter=(bX=True,bY=True,bZ=True)
    Parents=((Item=(Type=Bone,Name="dyn_skirt_rear_l_02"),Weight=1))
    Weight=1
    WorldUp=(Target=(X=0,Y=0,Z=1),Kind=Location,Space=(Type=Bone,Name="dyn_skirt_rear_l_02"))
   )
   &RigUnit_AimConstraintLocalSpaceOffset(
    Child=(Type=Bone,Name="dyn_leg_armor_rear_r_MSTR")
    ExecutePin=()
    Filter=(bX=True,bY=True,bZ=True)
    Parents=((Item=(Type=Bone,Name="dyn_skirt_rear_r_02"),Weight=1))
    Weight=1
    WorldUp=(Target=(X=0,Y=0,Z=1),Kind=Location,Space=(Type=Bone,Name="dyn_skirt_rear_r_02"))
   )
  ]}
 New Function_ContainedGraph {nodes [
   RigVMFunctionEntryNode()
   > RigVMFunction_Sequence = RigVMAggregateNode()
   n1 = &RigVMFunction_ControlFlowBranch(ExecuteContext=@RigVMFunction_Sequence.B, Condition=RigVMVariableNode() {variable=Anim Dynamics Cape}.Value)
   RigVMFunction_ControlFlowBranch = &RigVMFunction_ControlFlowBranch(ExecuteContext=@n1.False, Condition=RigVMVariableNode() {variable=WearingBackbling}.Value)
   n2 = &RigUnit_SphericalPoseReader(
    ActiveRegionScaleFactors=(PositiveWidth=1,NegativeWidth=3,PositiveHeight=3,NegativeHeight=3)
    ActiveRegionSize=0.1
    Debug=(bDrawDebug=false,bDraw2D=False,bDrawLocalAxes=False,DebugScale=25,DebugSegments=20,DebugThickness=0.25)
    DriverAxis=(X=1,Y=0,Z=0)
    DriverItem=(Type=Bone,Name="spine_05")
    FalloffRegionScaleFactors=(PositiveWidth=1,NegativeWidth=1,PositiveHeight=1,NegativeHeight=1)
    FalloffSize=0.3
    FlipHeightScaling=False
    FlipWidthScaling=False
    OptionalParentItem=(Type=Bone,Name="pelvis")
    RotationOffset=(X=90,Y=0,Z=-20)
    ExecutePin=@RigVMFunction_ControlFlowBranch.False
   )
   > &RigUnit_ModifyTransforms(
    ItemToModify=((Item=(Type=Bone,Name="CR_cape_aimat"),Transform=(Rotation=(X=0,Y=0,Z=0,W=1),Translation=(X=0,Y=-45,Z=0),Scale3D=(X=1,Y=1,Z=1))))
    WeightMinimum=0
    Weight=@n2.OutputParam
   )
   RigUnit_GetRelativeTransformForItem = &RigUnit_GetRelativeTransformForItem(Child=(Type=Bone,Name="spine_05"), Parent=(Type=Bone,Name="pelvis"), bChildInitial=False, bParentInitial=False)
   RigVMFunction_MathQuaternionSwingTwist = &RigVMFunction_MathQuaternionSwingTwist(TwistAxis=(X=1,Y=0,Z=0), Input=@RigUnit_GetRelativeTransformForItem.RelativeTransform.Rotation)
   > RigUnit_SetRotation = &RigUnit_SetRotation(Item=(Type=Bone,Name="CR_cape_rotate"), Weight=1, bInitial=False, Value=@RigVMFunction_MathQuaternionSwingTwist.Twist)
   > &RigUnit_AimConstraintLocalSpaceOffset(
    Child=(Type=Bone,Name="dyn_cape_MSTR")
    Filter=(bX=True,bY=True,bZ=True)
    Parents=((Item=(Type=Bone,Name="CR_cape_aimat"),Weight=1))
    Weight=1
    WorldUp=(Target=(X=0,Y=0,Z=1),Kind=Location,Space=(Type=Bone,Name="CR_cape_aimat"))
   )
   > &RigUnit_AimConstraintLocalSpaceOffset(
    Child=(Type=Bone,Name="dyn_cape_l_MSTR")
    Filter=(bX=True,bY=True,bZ=True)
    Parents=((Item=(Type=Bone,Name="CR_cape_aimat"),Weight=1))
    Weight=1
    WorldUp=(Target=(X=0,Y=0,Z=1),Kind=Location,Space=(Type=Bone,Name="CR_cape_aimat"))
   )
   > &RigUnit_AimConstraintLocalSpaceOffset(
    Child=(Type=Bone,Name="dyn_cape_r_MSTR")
    Filter=(bX=True,bY=True,bZ=True)
    Parents=((Item=(Type=Bone,Name="CR_cape_aimat"),Weight=1))
    Weight=1
    WorldUp=(Target=(X=0,Y=0,Z=1),Kind=Location,Space=(Type=Bone,Name="CR_cape_aimat"))
   )
   &RigUnit_ModifyTransforms(
    ItemToModify=((Item=(Type=Bone,Name="NewBone"),Transform=(Rotation=(X=0,Y=0,Z=0,W=1),Translation=(X=0,Y=-2,Z=0),Scale3D=(X=1,Y=1,Z=1))))
    Weight=1
    WeightMinimum=0
    ExecutePin=@RigVMFunction_ControlFlowBranch.True
   )
   > &RigUnit_AimConstraintLocalSpaceOffset(
    Child=(Type=Bone,Name="dyn_cape_MSTR")
    Filter=(bX=True,bY=True,bZ=True)
    Parents=((Item=(Type=Bone,Name="NewBone"),Weight=1))
    Weight=1
    WorldUp=(Target=(X=0,Y=0,Z=1),Kind=Location,Space=(Type=Bone,Name="NewBone"))
   )
   > &RigUnit_AimConstraintLocalSpaceOffset(
    Child=(Type=Bone,Name="dyn_cape_l_MSTR")
    Filter=(bX=True,bY=True,bZ=True)
    Parents=((Item=(Type=Bone,Name="NewBone"),Weight=1))
    Weight=1
    WorldUp=(Target=(X=0,Y=0,Z=1),Kind=Location,Space=(Type=Bone,Name="NewBone"))
   )
   > &RigUnit_AimConstraintLocalSpaceOffset(
    Child=(Type=Bone,Name="dyn_cape_r_MSTR")
    Filter=(bX=True,bY=True,bZ=True)
    Parents=((Item=(Type=Bone,Name="NewBone"),Weight=1))
    Weight=1
    WorldUp=(Target=(X=0,Y=0,Z=1),Kind=Location,Space=(Type=Bone,Name="NewBone"))
   )
   n3 = &RigUnit_SphericalPoseReader(
    ActiveRegionScaleFactors=(PositiveWidth=1.5,NegativeWidth=3,PositiveHeight=3,NegativeHeight=3)
    ActiveRegionSize=0.1
    Debug=(bDrawDebug=true,bDraw2D=False,bDrawLocalAxes=False,DebugScale=25,DebugSegments=20,DebugThickness=0.25)
    DriverAxis=(X=1,Y=0,Z=0)
    DriverItem=(Type=Bone,Name="spine_05")
    FalloffRegionScaleFactors=(PositiveWidth=1,NegativeWidth=1,PositiveHeight=1,NegativeHeight=1)
    FalloffSize=0.3
    FlipHeightScaling=False
    FlipWidthScaling=false
    OptionalParentItem=(Type=Bone,Name="pelvis")
    RotationOffset=(X=-90,Y=0,Z=45)
    ExecutePin=@n1.True
   )
   > &RigUnit_ModifyTransforms(
    ItemToModify="((Item=(Type=Bone,Name=\"CR_cape_aimat\"),Transform=(Rotation=(X=0,Y=0,Z=0,W=1),Translation=(X=0,Y=-45,Z=0),Scale3D=(X=1,Y=1,Z=1))),(Item=(Type=Bone,Name=\"dyn_cape_02\"),Transform=(Rotation=(X=0,Y=-0.3827,Z=0,W=0.9239),Translation=(X=0,Y=0,Z=0),Scale3D=(X=1,Y ...(truncated)"
    WeightMinimum=0
    Weight=@n3.OutputParam
   )
   > &RigUnit_ModifyTransforms(
    ItemToModify=((Item=(Type=Bone,Name="NewBone"),Transform=(Rotation=(X=0,Y=0,Z=0,W=1),Translation=(X=0,Y=-4,Z=0),Scale3D=(X=1,Y=1,Z=1))))
    Weight=1
    WeightMinimum=0
   )
   > &RigUnit_AimConstraintLocalSpaceOffset(
    Child=(Type=Bone,Name="dyn_cape_MSTR")
    Filter=(bX=True,bY=True,bZ=True)
    Parents=((Item=(Type=Bone,Name="NewBone"),Weight=1))
    Weight=1
    WorldUp=(Target=(X=0,Y=0,Z=1),Kind=Location,Space=(Type=Bone,Name="NewBone"))
   )
   > &RigUnit_AimConstraintLocalSpaceOffset(
    Child=(Type=Bone,Name="dyn_cape_l_MSTR")
    Filter=(bX=True,bY=True,bZ=True)
    Parents=((Item=(Type=Bone,Name="NewBone"),Weight=1))
    Weight=1
    WorldUp=(Target=(X=0,Y=0,Z=1),Kind=Location,Space=(Type=Bone,Name="NewBone"))
   )
   > &RigUnit_AimConstraintLocalSpaceOffset(
    Child=(Type=Bone,Name="dyn_cape_r_MSTR")
    Filter=(bX=True,bY=True,bZ=True)
    Parents=((Item=(Type=Bone,Name="NewBone"),Weight=1))
    Weight=1
    WorldUp=(Target=(X=0,Y=0,Z=1),Kind=Location,Space=(Type=Bone,Name="NewBone"))
   )
   n4 = &RigUnit_SphericalPoseReader(
    ActiveRegionScaleFactors=(PositiveWidth=2,NegativeWidth=2,PositiveHeight=0.5,NegativeHeight=2)
    ActiveRegionSize=0.1
    Debug=(bDrawDebug=false,bDraw2D=False,bDrawLocalAxes=False,DebugScale=25,DebugSegments=20,DebugThickness=0.25)
    DriverAxis=(X=1,Y=0,Z=0)
    DriverItem=(Type=Bone,Name="clavicle_l")
    FalloffRegionScaleFactors=(PositiveWidth=1,NegativeWidth=1,PositiveHeight=0.5,NegativeHeight=1)
    FalloffSize=0.3
    FlipHeightScaling=False
    FlipWidthScaling=False
    OptionalParentItem=(Type=Bone,Name="spine_04")
    RotationOffset=(X=0,Y=-70,Z=-27)
    ExecutePin=@RigVMFunction_Sequence.C
   )
   > &RigUnit_ModifyTransforms(
    ItemToModify="((Item=(Type=Bone,Name=\"dyn_cape_clavicle_l\"),Transform=(Rotation=(X=0,Y=0,Z=-0.0349,W=0.9994),Translation=(X=0,Y=-2,Z=0),Scale3D=(X=1,Y=1,Z=1))),(Item=(Type=Bone,Name=\"dyn_cape_l_MSTR\"),Transform=(Rotation=(X=0,Y=0,Z=0.0872,W=0.9962),Translation=(X=0,Y=0, ...(truncated)"
    WeightMinimum=0
    Weight=@n4.OutputParam
   )
   > n5 = &RigUnit_SphericalPoseReader(
    ActiveRegionScaleFactors=(PositiveWidth=2,NegativeWidth=2,PositiveHeight=0.5,NegativeHeight=2)
    ActiveRegionSize=0.1
    Debug=(bDrawDebug=false,bDraw2D=False,bDrawLocalAxes=False,DebugScale=25,DebugSegments=20,DebugThickness=0.25)
    DriverAxis=(X=-1,Y=0,Z=0)
    DriverItem=(Type=Bone,Name="clavicle_r")
    FalloffRegionScaleFactors=(PositiveWidth=1,NegativeWidth=1,PositiveHeight=0.5,NegativeHeight=1)
    FalloffSize=0.3
    FlipHeightScaling=true
    FlipWidthScaling=False
    OptionalParentItem=(Type=Bone,Name="spine_04")
    RotationOffset=(X=0,Y=110,Z=-27)
   )
   > &RigUnit_ModifyTransforms(
    ItemToModify="((Item=(Type=Bone,Name=\"dyn_cape_clavicle_r\"),Transform=(Rotation=(X=0,Y=0,Z=-0.0349,W=0.9994),Translation=(X=0,Y=0,Z=0),Scale3D=(X=1,Y=1,Z=1))),(Item=(Type=Bone,Name=\"dyn_cape_r_MSTR\"),Transform=(Rotation=(X=0,Y=0,Z=0,W=1),Translation=(X=0,Y=0,Z=-3),Scale ...(truncated)"
    WeightMinimum=0
    Weight=@n5.OutputParam
   )
   RigVMFunctionReturnNode(ExecuteContext=@RigVMFunction_Sequence.D)
   &RigUnit_ParentConstraint(
    AdvancedSettings=(RotationOrderForFilter=XZY)
    Child=(Type=Bone,Name="dyn_cape_MSTR")
    ExecutePin=()
    Filter=(TranslationFilter=(bX=True,bY=True,bZ=True),RotationFilter=(bX=false,bY=false,bZ=false),ScaleFilter=(bX=True,bY=True,bZ=True))
    Parents="((Item=(Type=Bone,Name=\"deform_upperarm_up_r\"),Weight=0.0001),(Item=(Type=Bone,Name=\"upperarm_twist_01_l\"),Weight=0.0002),(Item=(Type=Bone,Name=\"upperarm_twist_01_r\"),Weight=0.0003),(Item=(Type=Bone,Name=\"clavicle_l\"),Weight=0.0076),(Item=(Type=Bone,Name=\" ...(truncated)"
    Weight=1
   )
   > &RigUnit_ParentConstraint(
    AdvancedSettings=(RotationOrderForFilter=XZY)
    Child=(Type=Bone,Name="dyn_cape_l_MSTR")
    Filter=(TranslationFilter=(bX=True,bY=True,bZ=True),RotationFilter=(bX=false,bY=false,bZ=false),ScaleFilter=(bX=True,bY=True,bZ=True))
    Parents="((Item=(Type=Bone,Name=\"spine_04\"),Weight=0.016),(Item=(Type=Bone,Name=\"upperarm_twist_02_l\"),Weight=0.0171),(Item=(Type=Bone,Name=\"spine_03\"),Weight=0.0535),(Item=(Type=Bone,Name=\"spine_05\"),Weight=0.0771),(Item=(Type=Bone,Name=\"deform_upperarm_bk_l\"),Wei ...(truncated)"
    Weight=1
   )
   > &RigUnit_ParentConstraint(
    AdvancedSettings=(RotationOrderForFilter=XZY)
    Child=(Type=Bone,Name="dyn_cape_r_MSTR")
    Filter=(TranslationFilter=(bX=True,bY=True,bZ=True),RotationFilter=(bX=false,bY=false,bZ=false),ScaleFilter=(bX=True,bY=True,bZ=True))
    Parents="((Item=(Type=Bone,Name=\"spine_04\"),Weight=0.016),(Item=(Type=Bone,Name=\"upperarm_twist_02_r\"),Weight=0.0171),(Item=(Type=Bone,Name=\"spine_03\"),Weight=0.0535),(Item=(Type=Bone,Name=\"spine_05\"),Weight=0.0771),(Item=(Type=Bone,Name=\"deform_upperarm_bk_r\"),Wei ...(truncated)"
    Weight=1
   )
   &RigUnit_ModifyTransforms(
    ExecutePin=()
    ItemToModify="((Item=(Type=Bone,Name=\"dyn_cape_l_MSTR\"),Transform=(Rotation=(X=0,Y=0,Z=0,W=1),Translation=(X=0,Y=0,Z=1),Scale3D=(X=1,Y=1,Z=1))),(Item=(Type=Bone,Name=\"dyn_cape_r_MSTR\"),Transform=(Rotation=(X=0,Y=0,Z=0,W=1),Translation=(X=0,Y=0,Z=-1),Scale3D=(X=1,Y=1,Z=1 ...(truncated)"
    Weight=1
    WeightMinimum=0
   )
  ]}
 New Function_ContainedGraph {nodes [
   RigVMFunctionEntryNode()
   > &RigUnit_ParentConstraint(
    AdvancedSettings=(RotationOrderForFilter=XZY)
    Child=(Type=Bone,Name="dyn_cape_tie_l")
    Parents="((Item=(Type=Bone,Name=\"upperarm_twist_01_l\"),Weight=0.0091),(Item=(Type=Bone,Name=\"deform_upperarm_fr_l\"),Weight=0.0093),(Item=(Type=Bone,Name=\"deform_upperarm_up_l\"),Weight=0.0168),(Item=(Type=Bone,Name=\"spine_03\"),Weight=0.0465),(Item=(Type=Bone,Name=\"d ...(truncated)"
    Weight=1
   )
   > &RigUnit_ParentConstraint(
    AdvancedSettings=(RotationOrderForFilter=XZY)
    Child=(Type=Bone,Name="dyn_cape_tie_mid")
    Parents="((Item=(Type=Bone,Name=\"deform_pec_l\"),Weight=0.0059),(Item=(Type=Bone,Name=\"deform_pec_r\"),Weight=0.0061),(Item=(Type=Bone,Name=\"clavicle_l\"),Weight=0.0198),(Item=(Type=Bone,Name=\"clavicle_r\"),Weight=0.0203),(Item=(Type=Bone,Name=\"spine_02\"),Weight=0.0328 ...(truncated)"
    Weight=1
   )
   > &RigUnit_ParentConstraint(
    AdvancedSettings=(RotationOrderForFilter=XZY)
    Child=(Type=Bone,Name="dyn_cape_tie_mid_l")
    Parents="((Item=(Type=Bone,Name=\"neck_01\"),Weight=0.0044),(Item=(Type=Bone,Name=\"neck_02\"),Weight=0.0109),(Item=(Type=Bone,Name=\"spine_02\"),Weight=0.0245),(Item=(Type=Bone,Name=\"spine_03\"),Weight=0.0764),(Item=(Type=Bone,Name=\"deform_pec_l\"),Weight=0.1098),(Item=(T ...(truncated)"
    Weight=1
   )
   > &RigUnit_ParentConstraint(
    AdvancedSettings=(RotationOrderForFilter=XZY)
    Child=(Type=Bone,Name="dyn_cape_tie_r")
    Parents="((Item=(Type=Bone,Name=\"upperarm_twist_01_r\"),Weight=0.0096),(Item=(Type=Bone,Name=\"deform_upperarm_fr_r\"),Weight=0.0108),(Item=(Type=Bone,Name=\"deform_upperarm_up_r\"),Weight=0.0193),(Item=(Type=Bone,Name=\"spine_03\"),Weight=0.0467),(Item=(Type=Bone,Name=\"d ...(truncated)"
    Weight=1
   )
   > &RigUnit_ParentConstraint(
    AdvancedSettings=(RotationOrderForFilter=XZY)
    Child=(Type=Bone,Name="dyn_cape_tie_mid_r")
    Parents="((Item=(Type=Bone,Name=\"neck_01\"),Weight=0.0046),(Item=(Type=Bone,Name=\"neck_02\"),Weight=0.0112),(Item=(Type=Bone,Name=\"spine_02\"),Weight=0.0244),(Item=(Type=Bone,Name=\"spine_03\"),Weight=0.076),(Item=(Type=Bone,Name=\"deform_pec_r\"),Weight=0.1094),(Item=(Ty ...(truncated)"
    Weight=1
   )
   > RigVMFunctionReturnNode()
  ]}
 New Function_ContainedGraph {nodes [
   RigVMFunctionEntryNode()
   > RigVMFunction_Sequence = RigVMAggregateNode()
   RigVMFunctionReturnNode(ExecuteContext=@RigVMFunction_Sequence.E)
   RigUnit_GetTransform = &RigUnit_GetTransform(Item=(Type=Bone,Name="upperarm_l"), Space=LocalSpace, bInitial=False)
   RigVMFunction_MathQuaternionSwingTwist = &RigVMFunction_MathQuaternionSwingTwist(TwistAxis=(X=1,Y=0,Z=0), Input=@RigUnit_GetTransform.Transform.Rotation)
   Inverse = /Script/RigVM.RigVMFunction_MathQuaternionInverse(Value=@RigVMFunction_MathQuaternionSwingTwist.Twist) {is=struct RigVMUnitNode, template="Inverse::Execute(in Value,out Result)"}
   &RigUnit_SetRotation(Item=(Type=Bone,Name="dyn_shoulder_armor_l"), Weight=0.5, bInitial=False, Value=@Inverse.Result, ExecutePin=@RigVMFunction_Sequence.A)
   n1 = &RigUnit_GetTransform(Item=(Type=Bone,Name="upperarm_r"), Space=LocalSpace, bInitial=False)
   > RigUnit_SetRotation = &RigUnit_SetRotation(
    Item=(Type=Bone,Name="dyn_shoulder_armor_r")
    Weight=0.5
    bInitial=False
    Value=&RigVMFunction_MathQuaternionInverse(Value=&RigVMFunction_MathQuaternionSwingTwist(TwistAxis=(X=1,Y=0,Z=0), Input=@n1.Transform.Rotation).Twist) {template="Inverse::Execute(in Value,out Result)"}.Result
   )
   n2 = &RigUnit_SphericalPoseReader(
    ActiveRegionScaleFactors=(PositiveWidth=1,NegativeWidth=1,PositiveHeight=1,NegativeHeight=1)
    ActiveRegionSize=0.1
    Debug=(bDrawDebug=false,bDraw2D=False,bDrawLocalAxes=False,DebugScale=25,DebugSegments=20,DebugThickness=0.25)
    DriverAxis=(X=1,Y=0,Z=0)
    DriverItem=(Type=Bone,Name="clavicle_l")
    FalloffRegionScaleFactors=(PositiveWidth=1,NegativeWidth=1,PositiveHeight=1,NegativeHeight=1)
    FalloffSize=0.22
    FlipHeightScaling=False
    FlipWidthScaling=False
    OptionalParentItem=(Type=Bone,Name="None")
    RotationOffset=(X=0,Y=-50,Z=0)
    ExecutePin=@RigVMFunction_Sequence.B
   )
   > &RigUnit_ModifyTransforms(
    ItemToModify=((Item=(Type=Bone,Name="dyn_shoulder_armor_l"),Transform=(Rotation=(X=0,Y=0,Z=0,W=1),Translation=(X=5,Y=0,Z=0),Scale3D=(X=0.9,Y=1,Z=1))))
    WeightMinimum=0
    Weight=@n2.OutputParam
   )
   > n3 = &RigUnit_SphericalPoseReader(
    ActiveRegionScaleFactors=(PositiveWidth=1,NegativeWidth=1,PositiveHeight=1,NegativeHeight=1)
    ActiveRegionSize=0.1
    Debug=(bDrawDebug=false,bDraw2D=False,bDrawLocalAxes=False,DebugScale=25,DebugSegments=20,DebugThickness=0.25)
    DriverAxis=(X=-1,Y=0,Z=0)
    DriverItem=(Type=Bone,Name="clavicle_r")
    FalloffRegionScaleFactors=(PositiveWidth=1,NegativeWidth=1,PositiveHeight=1,NegativeHeight=1)
    FalloffSize=0.22
    FlipHeightScaling=False
    FlipWidthScaling=False
    OptionalParentItem=(Type=Bone,Name="None")
    RotationOffset=(X=180,Y=-50,Z=0)
   )
   > &RigUnit_ModifyTransforms(
    ItemToModify=((Item=(Type=Bone,Name="dyn_shoulder_armor_r"),Transform=(Rotation=(X=0,Y=0,Z=0,W=1),Translation=(X=-6,Y=0,Z=0),Scale3D=(X=0.9,Y=1,Z=1))))
    WeightMinimum=0
    Weight=@n3.OutputParam
   )
   n4 = &RigUnit_SphericalPoseReader(
    ActiveRegionScaleFactors=(PositiveWidth=2.5,NegativeWidth=3,PositiveHeight=3,NegativeHeight=3)
    ActiveRegionSize=0.1
    Debug=(bDrawDebug=false,DebugScale=25,DebugSegments=20,DebugThickness=0.25)
    DriverAxis=(X=1,Y=0,Z=0)
    DriverItem=(Type=Bone,Name="upperarm_l")
    FalloffRegionScaleFactors=(PositiveWidth=1,NegativeWidth=1,PositiveHeight=1,NegativeHeight=1)
    FalloffSize=0.5
    OptionalParentItem=(Type=Bone)
    RotationOffset=(X=0,Y=0,Z=0)
    ExecutePin=@RigVMFunction_Sequence.C
   )
   > &RigUnit_ModifyTransforms(
    ItemToModify=((Item=(Type=Bone,Name="dyn_shoulder_armor_l"),Transform=(Rotation=(X=0.2588,Y=0,Z=0,W=0.9659),Translation=(X=0,Y=0,Z=0),Scale3D=(X=0.9,Y=1,Z=1))))
    WeightMinimum=0
    Weight=@n4.OutputParam
   )
   > n5 = &RigUnit_SphericalPoseReader(
    ActiveRegionScaleFactors=(PositiveWidth=2.5,NegativeWidth=3,PositiveHeight=3,NegativeHeight=3)
    ActiveRegionSize=0.1
    Debug=(bDrawDebug=false,DebugScale=25,DebugSegments=20,DebugThickness=0.25)
    DriverAxis=(X=-1,Y=0,Z=0)
    DriverItem=(Type=Bone,Name="upperarm_r")
    FalloffRegionScaleFactors=(PositiveWidth=1,NegativeWidth=1,PositiveHeight=1,NegativeHeight=1)
    FalloffSize=0.5
    FlipHeightScaling=true
    OptionalParentItem=(Type=Bone)
    RotationOffset=(X=180,Y=0,Z=180)
   )
   > &RigUnit_ModifyTransforms(
    ItemToModify=((Item=(Type=Bone,Name="dyn_shoulder_armor_r"),Transform=(Rotation=(X=0.2588,Y=0,Z=0,W=0.9659),Translation=(X=0,Y=0,Z=0),Scale3D=(X=0.9,Y=1,Z=1))))
    WeightMinimum=0
    Weight=@n5.OutputParam
   )
   n6 = &RigUnit_SphericalPoseReader(
    ActiveRegionScaleFactors=(PositiveWidth=0.5,NegativeWidth=3,PositiveHeight=1.5,NegativeHeight=1.5)
    ActiveRegionSize=0.1
    Debug=(bDrawDebug=false,DebugScale=25,DebugSegments=20,DebugThickness=0.25)
    DriverAxis=(X=1,Y=0,Z=0)
    DriverItem=(Type=Bone,Name="upperarm_l")
    FalloffRegionScaleFactors=(PositiveWidth=1,NegativeWidth=1,PositiveHeight=1,NegativeHeight=1)
    FalloffSize=0.4
    OptionalParentItem=(Type=Bone)
    RotationOffset=(X=90,Y=0,Z=0)
    ExecutePin=@RigVMFunction_Sequence.D
   )
   > &RigUnit_ModifyTransforms(
    ItemToModify=((Item=(Type=Bone,Name="dyn_shoulder_armor_l"),Transform=(Rotation=(X=0,Y=-0.0436,Z=0,W=0.999),Translation=(X=0,Y=-1,Z=0),Scale3D=(X=1,Y=1,Z=1))))
    WeightMinimum=0
    Weight=@n6.OutputParam
   )
   > n7 = &RigUnit_SphericalPoseReader(
    ActiveRegionScaleFactors=(PositiveWidth=0.5,NegativeWidth=3,PositiveHeight=1.5,NegativeHeight=1.5)
    ActiveRegionSize=0.1
    Debug=(bDrawDebug=false,DebugScale=25,DebugSegments=20,DebugThickness=0.25)
    DriverAxis=(X=-1,Y=0,Z=0)
    DriverItem=(Type=Bone,Name="upperarm_r")
    FalloffRegionScaleFactors=(PositiveWidth=1,NegativeWidth=1,PositiveHeight=1,NegativeHeight=1)
    FalloffSize=0.4
    FlipWidthScaling=true
    OptionalParentItem=(Type=Bone)
    RotationOffset=(X=90,Y=180,Z=180)
   )
   > &RigUnit_ModifyTransforms(
    ItemToModify=((Item=(Type=Bone,Name="dyn_shoulder_armor_r"),Transform=(Rotation=(X=0,Y=-0.0436,Z=0,W=0.999),Translation=(X=0,Y=1,Z=0),Scale3D=(X=1,Y=1,Z=1))))
    WeightMinimum=0
    Weight=@n7.OutputParam
   )
  ]}
 RigVMFunctionLibrary {nodes [
   Deform Armor = RigVMCollapseNode(ExecuteContext=())
   Deform Cape = RigVMCollapseNode(ExecuteContext=())
   Deform Cape Ties = RigVMCollapseNode(ExecuteContext=())
   Deform Shoulder Armor = RigVMCollapseNode(ExecuteContext=())
  ]}
 RigVMFunction_Sequence_1_ContainedGraph {nodes [
   n1 = RigVMFunctionEntryNode()
   RigVMFunction_Sequence = &RigVMFunction_Sequence(ExecuteContext=@n1.ExecuteContext)
   n2 = &RigVMFunction_Sequence(ExecuteContext=@RigVMFunction_Sequence.B)
   n3 = &RigVMFunction_Sequence(ExecuteContext=@n2.B)
   RigVMFunctionReturnNode(A=@RigVMFunction_Sequence.A, B=@n2.A, C=@n3.A, D=@n3.B)
  ]}
 RigVMFunction_Sequence_1_ContainedGraph {nodes [
   n1 = RigVMFunctionEntryNode()
   RigVMFunction_Sequence = &RigVMFunction_Sequence(ExecuteContext=@n1.ExecuteContext)
   n2 = &RigVMFunction_Sequence(ExecuteContext=@RigVMFunction_Sequence.B)
   n3 = &RigVMFunction_Sequence(ExecuteContext=@n2.B)
   n4 = &RigVMFunction_Sequence(ExecuteContext=@n3.B)
   RigVMFunctionReturnNode(A=@RigVMFunction_Sequence.A, B=@n2.A, C=@n3.A, D=@n4.A, E=@n4.B)
  ]}
 RigVMFunction_Sequence_1_ContainedGraph {nodes [
   n1 = RigVMFunctionEntryNode()
   RigVMFunction_Sequence = &RigVMFunction_Sequence(ExecuteContext=@n1.ExecuteContext)
   n2 = &RigVMFunction_Sequence(ExecuteContext=@RigVMFunction_Sequence.B)
   n3 = &RigVMFunction_Sequence(ExecuteContext=@n2.B)
   n4 = &RigVMFunction_Sequence(ExecuteContext=@n3.B)
   RigVMFunctionReturnNode(A=@RigVMFunction_Sequence.A, B=@n2.A, C=@n3.A, D=@n4.A, E=@n4.B)
  ]}
 RigVMModel {nodes [
   RigUnit_BeginExecution = &RigUnit_BeginExecution()
   Deform Armor = RigVMFunctionReferenceNode(ExecuteContext=@RigUnit_BeginExecution.ExecutePin)
   > Deform Cape = RigVMFunctionReferenceNode()
   > Deform Cape Ties = RigVMFunctionReferenceNode()
   > Deform Shoulder Armor = RigVMFunctionReferenceNode()
  ]}
]
hierarchy [
 root {loc=<0 0 0>, rot=<0 0 0 1>}
 pelvis {parent=root, loc=<0 -0.7071 93.1191>, rot=<0.0127 0.707 -0.0127 -0.707>}
 spine_01 {parent=pelvis, loc=<6.7494 -0.6388 0>, rot=<0 0 0.034 -0.9994>}
 spine_02 {parent=spine_01, loc=<7.252 -0.035 0>, rot=<0 0 0.0238 -0.9997>}
 spine_03 {parent=spine_02, loc=<6.0106 0 0>, rot=<0 0 0.0149 -0.9999>}
 spine_04 {parent=spine_03, loc=<8.1023 0 0>, rot=<0 0 -0.0132 -0.9999>}
 spine_05 {parent=spine_04, loc=<10.0142 0 0>, rot=<0 0 -0.0015 -1>}
 neck_01 {parent=spine_05, loc=<11.3138 0.1529 0>, rot=<0 0 -0.2003 -0.9797>}
 neck_02 {parent=neck_01, loc=<4.5627 0 0>, rot=<0 0 0.0504 -0.9987>}
 head {parent=neck_02, loc=<5.3914 0 0>, rot=<0 0 0.115 -0.9934>}
 clavicle_l {parent=spine_05, loc=<4.3575 -0.1466 -2.3223>, rot=<0.0465 -0.7289 -0.009 -0.6829>}
 upperarm_l {parent=clavicle_l, loc=<15.5111 0 0>, rot=<0.0226 -0.3234 -0.0088 -0.9459>}
 lowerarm_l {parent=upperarm_l, loc=<26.6514 0 0>, rot=<0.0391 -0.0099 -0.2453 -0.9686>}
 lowerarm_twist_02_l {parent=lowerarm_l, loc=<9.6199 0 0>, rot=<0 0 0 -1>}
 deform_elbow_in_base_l {parent=lowerarm_l, loc=<0.2112 5.259 0.8274>, rot=<0 0 0 -1>}
 deform_elbow_in_l {parent=deform_elbow_in_base_l, loc=<0 0 0>, rot=<0 0 0 1>}
 deform_elbow_out_base_l {parent=lowerarm_l, loc=<0 -4 0>, rot=<0 0 0 -1>}
 deform_elbow_out_l {parent=deform_elbow_out_base_l, loc=<0 0 0>, rot=<0 0 0 1>}
 hand_l {parent=lowerarm_l, loc=<22.3491 0 0>, rot=<-0.7081 -0.0507 0.0975 -0.6975>}
 dyn_wrist_knot_l_01 {parent=hand_l, loc=<0.6136 3.8702 -1.3917>, rot=<0.6418 -0.3447 -0.208 -0.6527>}
 dyn_wrist_knot_l_02 {parent=hand_l, loc=<0.3156 3.5933 1.4486>, rot=<0.5846 0.1962 0.3996 -0.6783>}
 dyn_elbow_knot_l_01 {parent=lowerarm_l, loc=<8.3618 1.8462 7.4304>, rot=<0.0832 0.0766 -0.3502 -0.9298>}
 dyn_elbow_knot_l_02 {parent=lowerarm_l, loc=<7.9905 -0.6427 7.5613>, rot=<-0.0731 0.1169 0.5265 -0.8389>}
 upperarm_twist_01_l {parent=upperarm_l, loc=<0.0041 0 0>, rot=<0 0 0 -1>}
 deform_upperarm_bk_base_l {parent=upperarm_twist_01_l, loc=<1.0813 -7.2193 -1.9453>, rot=<0 0 0 1>}
 deform_upperarm_bk_l {parent=deform_upperarm_bk_base_l, loc=<0 0 0>, rot=<0 0 0 1>}
 deform_upperarm_cape_bk_l {parent=deform_upperarm_bk_l, loc=<0 0 0>, rot=<0 0 0 1>}
 deform_upperarm_fr_base_l {parent=upperarm_twist_01_l, loc=<-0.3691 7.0506 -2.765>, rot=<0 0 0 1>}
 deform_upperarm_fr_l {parent=deform_upperarm_fr_base_l, loc=<0 0 0>, rot=<0 0 0 1>}
 deform_upperarm_up_base_l {parent=upperarm_twist_01_l, loc=<-3.2216 -1.1391 3.3311>, rot=<0 0 0 1>}
 deform_upperarm_up_l {parent=deform_upperarm_up_base_l, loc=<0 0 0>, rot=<0 0 0 1>}
 upperarm_twist_02_l {parent=upperarm_l, loc=<13.7941 0 0>, rot=<0 0 0 -1>}
 dyn_shoulder_armor_l {parent=upperarm_l, loc=<6.7206 0 0>, rot=<0 0 0 -1>}
 dyn_cape_clavicle_l {parent=clavicle_l, loc=<0 0 0>, rot=<0 0 0 1>}
 clavicle_r {parent=spine_05, loc=<4.3575 -0.1466 2.3223>, rot=<0.7289 0.0465 -0.6829 0.009>}
 upperarm_r {parent=clavicle_r, loc=<-15.5111 0 0>, rot=<0.0226 -0.3234 -0.0088 -0.9459>}
 lowerarm_r {parent=upperarm_r, loc=<-26.6514 0 0>, rot=<0.0391 -0.0099 -0.2453 -0.9686>}
 lowerarm_twist_02_r {parent=lowerarm_r, loc=<-9.6199 0 0>, rot=<0 0 0 -1>}
 deform_elbow_in_base_r {parent=lowerarm_r, loc=<-0.2111 -5.2591 -0.8275>, rot=<0 0 0 -1>}
 deform_elbow_in_r {parent=deform_elbow_in_base_r, loc=<-0.0001 0 0.0002>, rot=<0 0 0 1>}
 deform_elbow_out_base_r {parent=lowerarm_r, loc=<0 4 -0.0001>, rot=<0 0 0 -1>}
 deform_elbow_out_r {parent=deform_elbow_out_base_r, loc=<0 0 0>, rot=<0 0 0 1>}
 hand_r {parent=lowerarm_r, loc=<-22.3491 0 0>, rot=<-0.7081 -0.0507 0.0975 -0.6975>}
 dyn_wrist_knot_r_01 {parent=hand_r, loc=<-0.6136 -3.8701 1.3917>, rot=<0.6418 -0.3447 -0.208 -0.6527>}
 dyn_wrist_knot_r_02 {parent=hand_r, loc=<-0.3158 -3.5932 -1.4487>, rot=<0.5846 0.1962 0.3996 -0.6783>}
 dyn_elbow_knot_r_01 {parent=lowerarm_r, loc=<-8.3617 -1.8463 -7.4304>, rot=<0.0832 0.0766 -0.3502 -0.9298>}
 dyn_elbow_knot_r_02 {parent=lowerarm_r, loc=<-7.9905 0.6427 -7.5614>, rot=<-0.0731 0.1169 0.5265 -0.8389>}
 upperarm_twist_01_r {parent=upperarm_r, loc=<-0.0041 0 0>, rot=<0 0 0 -1>}
 deform_upperarm_bk_base_r {parent=upperarm_twist_01_r, loc=<-1.0814 7.2192 1.9455>, rot=<0 0 0 1>}
 deform_upperarm_bk_r {parent=deform_upperarm_bk_base_r, loc=<0 0 0>, rot=<0 0 0 1>}
 deform_upperarm_cape_bk_r {parent=deform_upperarm_bk_r, loc=<0 0 0>, rot=<0 0 0 1>}
 deform_upperarm_fr_base_r {parent=upperarm_twist_01_r, loc=<0.3689 -7.0506 2.7652>, rot=<0 0 0 1>}
 deform_upperarm_fr_r {parent=deform_upperarm_fr_base_r, loc=<0 0 0>, rot=<0 0 0 1>}
 deform_upperarm_up_base_r {parent=upperarm_twist_01_r, loc=<3.2214 1.1391 -3.3308>, rot=<0 0 0 1>}
 deform_upperarm_up_r {parent=deform_upperarm_up_base_r, loc=<0 0 0>, rot=<0 0 0 -1>}
 upperarm_twist_02_r {parent=upperarm_r, loc=<-13.7941 0 0>, rot=<0 0 0 -1>}
 dyn_shoulder_armor_r {parent=upperarm_r, loc=<-6.7204 0 0>, rot=<0 0 0 -1>}
 dyn_cape_clavicle_r {parent=clavicle_r, loc=<0 0 0>, rot=<0 0 0 1>}
 attach_cape {parent=spine_05, loc=<9.017 -11.714 0>, rot=<0.0616 -0.7044 -0.0616 -0.7044>}
 dyn_cape_MSTR {parent=attach_cape, loc=<0 0.224 -0.3472>, rot=<-0.4855 -0.4855 0.5141 -0.5141>}
 dyn_cape_01 {parent=dyn_cape_MSTR, loc=<0 0 0>, rot=<0 0 0 1>}
 dyn_cape_02 {parent=dyn_cape_01, loc=<21 0 0>, rot=<0 0.0002 0 -1>}
 dyn_cape_03 {parent=dyn_cape_02, loc=<21 0 0>, rot=<0 0 0 1>}
 dyn_cape_04 {parent=dyn_cape_03, loc=<21 0 0>, rot=<0 0 0 1>}
 dyn_cape_05 {parent=dyn_cape_04, loc=<21 0 0>, rot=<0 0 0 1>}
 dyn_cape_06 {parent=dyn_cape_05, loc=<21 0 0>, rot=<0 0 0 1>}
 dyn_cape_l_MSTR {parent=attach_cape, loc=<12.6772 0.8153 0.0417>, rot=<-0.4817 -0.504 0.4544 -0.5546>}
 dyn_cape_l_01 {parent=dyn_cape_l_MSTR, loc=<0 0 0>, rot=<0 0 0 1>}
 dyn_cape_l_02 {parent=dyn_cape_l_01, loc=<21 0 0>, rot=<0 0.0006 0 -1>}
 dyn_cape_l_03 {parent=dyn_cape_l_02, loc=<21 0 0>, rot=<0 0 0 1>}
 dyn_cape_l_04 {parent=dyn_cape_l_03, loc=<21 0 0>, rot=<0 0 0 1>}
 dyn_cape_l_05 {parent=dyn_cape_l_04, loc=<21 0 0>, rot=<0 0 0 1>}
 dyn_cape_l_06 {parent=dyn_cape_l_05, loc=<21 0 0>, rot=<0 0 0 1>}
 dyn_cape_r_MSTR {parent=attach_cape, loc=<-12.6772 0.8153 0.0417>, rot=<0.5546 0.4544 0.504 -0.4817>}
 dyn_cape_r_01 {parent=dyn_cape_r_MSTR, loc=<0 0 0>, rot=<0 0 0 1>}
 dyn_cape_r_02 {parent=dyn_cape_r_01, loc=<-20.9999 0 0>, rot=<0 0.0006 0 -1>}
 dyn_cape_r_03 {parent=dyn_cape_r_02, loc=<-21.0001 0.0001 0>, rot=<0 0 0 1>}
 dyn_cape_r_04 {parent=dyn_cape_r_03, loc=<-21 0 0>, rot=<0 0 0 1>}
 dyn_cape_r_05 {parent=dyn_cape_r_04, loc=<-20.9999 -0.0001 0>, rot=<0 0 0 1>}
 dyn_cape_r_06 {parent=dyn_cape_r_05, loc=<-21 0.0001 -0.0001>, rot=<0 0 0 1>}
 attach_backpack {parent=spine_05, loc=<-0.123 -13.473 0>, rot=<-0.0093 -0.707 0.0093 -0.707>}
 dyn_backpack_collision {parent=attach_backpack, loc=<0 0 0>, rot=<0 0 0 1>}
 deform_pec_aim_base_l {parent=spine_04, loc=<7.8728 12.6388 -2.4123>, rot=<0.0102 -0.4602 0.3979 -0.7936>}
 deform_pec_aim_l {parent=deform_pec_aim_base_l, loc=<0 0 0>, rot=<0 0 0 1>}
 deform_pec_base_l {parent=deform_pec_aim_l, loc=<10.766 0 0>, rot=<0 0 0 1>}
 deform_pec_l {parent=deform_pec_base_l, loc=<0.055 0 0>, rot=<0 0 0 1>}
 deform_pec_aim_base_r {parent=spine_04, loc=<7.8723 12.6388 2.4123>, rot=<-0.4602 -0.0102 0.7936 0.3979>}
 deform_pec_aim_r {parent=deform_pec_aim_base_r, loc=<0 0 0>, rot=<0 0 0 1>}
 deform_pec_base_r {parent=deform_pec_aim_r, loc=<-10.7662 -0.0006 -0.0007>, rot=<0 0 0 1>}
 deform_pec_r {parent=deform_pec_base_r, loc=<0 0 0>, rot=<0 0 0 1>}
 dyn_cape_tie_l {parent=spine_04, loc=<18.8132 10.0109 -9.8527>, rot=<-0.2867 -0.8408 0.0605 0.4551>}
 dyn_cape_tie_bow_l_01 {parent=dyn_cape_tie_l, loc=<1.5183 -0.3665 1.534>, rot=<0 0.2999 0 -0.954>}
 dyn_cape_tie_bow_l_02 {parent=dyn_cape_tie_l, loc=<-0.9745 0.1735 2.3003>, rot=<0 0.7371 0 -0.6758>}
 dyn_cape_tie_mid {parent=spine_04, loc=<17.4272 10.1139 0>, rot=<-0.356 -0.9345 0 0>}
 dyn_cape_tie_mid_l {parent=spine_04, loc=<16.169 12.3445 -5.4463>, rot=<0.1706 0.7637 0.0995 -0.6146>}
 dyn_cape_tie_r {parent=spine_04, loc=<18.8136 10.011 9.8527>, rot=<0.8408 -0.2867 0.4551 -0.0605>}
 dyn_cape_tie_bow_r_01 {parent=dyn_cape_tie_r, loc=<-1.5185 0.3667 -1.5343>, rot=<0 0.2999 0 -0.954>}
 dyn_cape_tie_bow_r_02 {parent=dyn_cape_tie_r, loc=<0.9743 -0.1732 -2.3009>, rot=<0 0.7371 0 -0.6758>}
 dyn_cape_tie_mid_r {parent=spine_04, loc=<16.1695 12.3445 5.4463>, rot=<0.7637 -0.1706 0.6146 0.0995>}
 NewBone {parent=spine_03, loc=<0 0 0>, rot=<0 0 0 1>}
 dyn_leg_armor_side_l_MSTR {parent=spine_01, loc=<2.6763 3.6739 -12.5984>, rot=<-0.0914 0.5825 0.8066 0.0406>}
 dyn_leg_armor_side_l_01 {parent=dyn_leg_armor_side_l_MSTR, loc=<0 0 0>, rot=<0 0 0.0403 -0.9992>}
 dyn_leg_armor_side_l_02 {parent=dyn_leg_armor_side_l_01, loc=<10.6854 0 0>, rot=<0 0 -0.0131 -0.9999>}
 dyn_leg_armor_side_l_03 {parent=dyn_leg_armor_side_l_02, loc=<6.2058 0 0>, rot=<0 0 0 1>}
 dyn_leg_armor_side_l_04 {parent=dyn_leg_armor_side_l_03, loc=<7.3071 0 0>, rot=<0 0 0 1>}
 dyn_leg_armor_side_l_05 {parent=dyn_leg_armor_side_l_04, loc=<7.738 0 0>, rot=<0 0 0 1>}
 dyn_leg_armor_rear_l_MSTR {parent=spine_01, loc=<4.099 -9.1856 -5.8493>, rot=<0.2526 -0.9537 -0.155 -0.0514>}
 dyn_leg_armor_rear_l_01 {parent=dyn_leg_armor_rear_l_MSTR, loc=<0 0 0>, rot=<0 0 0 1>}
 dyn_leg_armor_rear_l_02 {parent=dyn_leg_armor_rear_l_01, loc=<3.3095 0 0>, rot=<0 0 -0.1425 -0.9898>}
 dyn_leg_armor_rear_l_03 {parent=dyn_leg_armor_rear_l_02, loc=<5.6254 0 0>, rot=<0 0 0 1>}
 dyn_leg_armor_rear_l_04 {parent=dyn_leg_armor_rear_l_03, loc=<5.3373 0 0>, rot=<0 0 0 1>}
 dyn_leg_armor_rear_l_05 {parent=dyn_leg_armor_rear_l_04, loc=<5.8556 0 0>, rot=<0 0 0 1>}
 dyn_leg_armor_front_MSTR {parent=spine_01, loc=<1.4106 10.5125 0>, rot=<0 0 0.9989 0.0463>}
 dyn_leg_armor_front_01 {parent=dyn_leg_armor_front_MSTR, loc=<0 0 0>, rot=<0 0 0 1>}
 dyn_leg_armor_front_02 {parent=dyn_leg_armor_front_01, loc=<4.9777 0 0>, rot=<0 0 -0.0217 -0.9998>}
 dyn_leg_armor_front_03 {parent=dyn_leg_armor_front_02, loc=<6.4877 0 0>, rot=<0 0 0 1>}
 dyn_leg_armor_front_04 {parent=dyn_leg_armor_front_03, loc=<6.2145 0 0>, rot=<0 0 0 1>}
 dyn_leg_armor_front_05 {parent=dyn_leg_armor_front_04, loc=<6.4583 0 0>, rot=<0 0 0 1>}
 dyn_leg_armor_rear_r_MSTR {parent=spine_01, loc=<4.0993 -9.1856 5.8493>, rot=<0.9537 0.2526 -0.0514 0.155>}
 dyn_leg_armor_rear_r_01 {parent=dyn_leg_armor_rear_r_MSTR, loc=<0 0 0>, rot=<0 0 0 1>}
 dyn_leg_armor_rear_r_02 {parent=dyn_leg_armor_rear_r_01, loc=<-3.3094 0.0001 0>, rot=<0 0 -0.1425 -0.9898>}
 dyn_leg_armor_rear_r_03 {parent=dyn_leg_armor_rear_r_02, loc=<-5.6258 -0.0001 0>, rot=<0 0 0 1>}
 dyn_leg_armor_rear_r_04 {parent=dyn_leg_armor_rear_r_03, loc=<-5.3373 -0.0001 0>, rot=<0 0 0 1>}
 dyn_leg_armor_rear_r_05 {parent=dyn_leg_armor_rear_r_04, loc=<-5.8556 0 0>, rot=<0 0 0 1>}
 dyn_leg_armor_side_r_MSTR {parent=spine_01, loc=<2.6759 3.6739 12.5984>, rot=<-0.5825 -0.0914 0.0406 -0.8066>}
 dyn_leg_armor_side_r_01 {parent=dyn_leg_armor_side_r_MSTR, loc=<0 0 0>, rot=<0 0 0.0403 -0.9992>}
 dyn_leg_armor_side_r_02 {parent=dyn_leg_armor_side_r_01, loc=<-10.6849 0.0001 0>, rot=<0 0 -0.0131 -0.9999>}
 dyn_leg_armor_side_r_03 {parent=dyn_leg_armor_side_r_02, loc=<-6.2058 0 0>, rot=<0 0 0 1>}
 dyn_leg_armor_side_r_04 {parent=dyn_leg_armor_side_r_03, loc=<-7.3072 0 0>, rot=<0 0 0 1>}
 dyn_leg_armor_side_r_05 {parent=dyn_leg_armor_side_r_04, loc=<-7.738 -0.0001 0>, rot=<0 0 0 1>}
 dyn_skirt_front_l_MSTR {parent=spine_01, loc=<3.2919 3.0874 -12.1108>, rot=<-0.0797 0.7026 0.7048 0.057>}
 dyn_skirt_front_l_01 {parent=dyn_skirt_front_l_MSTR, loc=<0 0 0>, rot=<0 0 0 1>}
 dyn_skirt_front_l_02 {parent=dyn_skirt_front_l_01, loc=<23.4716 0 0>, rot=<0 0 0 1>}
 dyn_skirt_front_l_03 {parent=dyn_skirt_front_l_02, loc=<23.0264 0 0>, rot=<0 0 0 1>}
 dyn_skirt_front_l_04 {parent=dyn_skirt_front_l_03, loc=<13.9362 0 0>, rot=<0 0 0 1>}
 dyn_skirt_front_l_05 {parent=dyn_skirt_front_l_04, loc=<13.9362 0 0>, rot=<0 0 0 1>}
 dyn_skirt_rear_l_MSTR {parent=spine_01, loc=<4.3762 -10.4189 -6.2626>, rot=<0.0631 -0.9911 -0.1155 -0.0202>}
 dyn_skirt_rear_l_01 {parent=dyn_skirt_rear_l_MSTR, loc=<0 0 0>, rot=<0 0 0 1>}
 dyn_skirt_rear_l_02 {parent=dyn_skirt_rear_l_01, loc=<28.7864 0 0>, rot=<0 0 0 1>}
 dyn_skirt_rear_l_03 {parent=dyn_skirt_rear_l_02, loc=<15.4859 0 0>, rot=<0 0 0 1>}
 dyn_skirt_rear_l_04 {parent=dyn_skirt_rear_l_03, loc=<13.9362 0 0>, rot=<0 0 0 1>}
 dyn_skirt_rear_l_05 {parent=dyn_skirt_rear_l_04, loc=<14.4794 0 0>, rot=<0 0 0 1>}
 dyn_skirt_side_l_MSTR {parent=spine_01, loc=<4.0093 -3.6409 -11.6979>, rot=<0.1034 -0.864 -0.4912 -0.0393>}
 dyn_skirt_side_l_01 {parent=dyn_skirt_side_l_MSTR, loc=<0 0 0>, rot=<0 0 0 1>}
 dyn_skirt_side_l_02 {parent=dyn_skirt_side_l_01, loc=<29.3289 0 0>, rot=<0 0 0 1>}
 dyn_skirt_side_l_03 {parent=dyn_skirt_side_l_02, loc=<15.5274 0 0>, rot=<0 0 0 1>}
 dyn_skirt_side_l_04 {parent=dyn_skirt_side_l_03, loc=<13.9362 0 0>, rot=<0 0 0 1>}
 dyn_skirt_side_l_05 {parent=dyn_skirt_side_l_04, loc=<14.3448 0 0>, rot=<0 0 0 1>}
 dyn_skirt_front_r_MSTR {parent=spine_01, loc=<3.2921 3.0875 12.1108>, rot=<-0.7026 -0.0797 0.057 -0.7048>}
 dyn_skirt_front_r_01 {parent=dyn_skirt_front_r_MSTR, loc=<0 0 0>, rot=<0 0 0 1>}
 dyn_skirt_front_r_02 {parent=dyn_skirt_front_r_01, loc=<-23.4718 0 0>, rot=<0 0 0 1>}
 dyn_skirt_front_r_03 {parent=dyn_skirt_front_r_02, loc=<-23.0264 0 0>, rot=<0 0 0 1>}
 dyn_skirt_front_r_04 {parent=dyn_skirt_front_r_03, loc=<-13.9363 0 0>, rot=<0 0 0 1>}
 dyn_skirt_front_r_05 {parent=dyn_skirt_front_r_04, loc=<-13.9362 0 0>, rot=<0 0 0 1>}
 dyn_skirt_rear_r_MSTR {parent=spine_01, loc=<4.3762 -10.4189 6.2626>, rot=<0.9911 0.0631 -0.0202 0.1155>}
 dyn_skirt_rear_r_01 {parent=dyn_skirt_rear_r_MSTR, loc=<0 0 0>, rot=<0 0 0 1>}
 dyn_skirt_rear_r_02 {parent=dyn_skirt_rear_r_01, loc=<-28.7863 0 0>, rot=<0 0 0 1>}
 dyn_skirt_rear_r_03 {parent=dyn_skirt_rear_r_02, loc=<-15.4859 0 0>, rot=<0 0 0 1>}
 dyn_skirt_rear_r_04 {parent=dyn_skirt_rear_r_03, loc=<-13.9363 0 0>, rot=<0 0 0 1>}
 dyn_skirt_rear_r_05 {parent=dyn_skirt_rear_r_04, loc=<-14.4793 0 0>, rot=<0 0 0 1>}
 dyn_skirt_side_r_MSTR {parent=spine_01, loc=<4.0094 -3.6409 11.6979>, rot=<0.864 0.1034 -0.0393 0.4912>}
 dyn_skirt_side_r_01 {parent=dyn_skirt_side_r_MSTR, loc=<0 0 0>, rot=<0 0 0 1>}
 dyn_skirt_side_r_02 {parent=dyn_skirt_side_r_01, loc=<-29.329 0 0>, rot=<0 0 0 1>}
 dyn_skirt_side_r_03 {parent=dyn_skirt_side_r_02, loc=<-15.5275 0 0>, rot=<0 0 0 1>}
 dyn_skirt_side_r_04 {parent=dyn_skirt_side_r_03, loc=<-13.9362 0 0>, rot=<0 0 0 1>}
 dyn_skirt_side_r_05 {parent=dyn_skirt_side_r_04, loc=<-14.3448 0 -0.0001>, rot=<0 0 0 1>}
 deform_pouch_l {parent=spine_01, loc=<4.7376 8.4334 -11.16>, rot=<-0.2607 -0.654 -0.2841 -0.6508>}
 deform_pouch_r {parent=spine_01, loc=<4.7376 8.4334 11.16>, rot=<0.2607 0.654 -0.2841 -0.6508>}
 thigh_l {parent=pelvis, loc=<-2.5693 -0.0109 -7.7501>, rot=<0.0012 0.0572 0.0155 -0.9982>}
 calf_l {parent=thigh_l, loc=<-39.0484 0 0>, rot=<0.0006 -0.0161 -0.0391 -0.9991>}
 foot_l {parent=calf_l, loc=<-42.561 0.1077 0.4292>, rot=<0.0418 -0.0455 0.0343 -0.9975>}
 calf_twist_02_l {parent=calf_l, loc=<-18.8064 -0.0967 0.0397>, rot=<0.0237 0.0095 -0.0035 -0.9997>}
 calf_twist_01_l {parent=calf_l, loc=<-35.813 -0.0967 0.0397>, rot=<0.0237 0.0095 -0.0035 -0.9997>}
 deform_knee_in_base_l {parent=calf_l, loc=<1.298 -5.2957 -0.0496>, rot=<0 0 0 -1>}
 deform_knee_in_l {parent=deform_knee_in_base_l, loc=<0 0 0>, rot=<0 0 0 1>}
 dyn_knee_knot_l_01 {parent=calf_l, loc=<-10.661 9.1263 -2.7669>, rot=<-0.3938 0.0911 0.9138 0.0414>}
 dyn_knee_knot_l_02 {parent=calf_l, loc=<-10.3954 8.9389 2.2903>, rot=<0.4532 -0.0854 0.8873 -0.0034>}
 thigh_twist_01_l {parent=thigh_l, loc=<-21.6303 0 0>, rot=<0.0013 0 0 -1>}
 deform_groin_base_l {parent=thigh_l, loc=<-2.1908 8.1142 -1.7442>, rot=<0 0 0 -1>}
 deform_groin_l {parent=deform_groin_base_l, loc=<0 0 0>, rot=<0 0 0 1>}
 thigh_r {parent=pelvis, loc=<-2.5693 -0.0109 7.7501>, rot=<0.0572 -0.0012 0.9982 0.0155>}
 calf_r {parent=thigh_r, loc=<39.0484 0 0>, rot=<0.0006 -0.0161 -0.0391 -0.9991>}
 foot_r {parent=calf_r, loc=<42.561 -0.1077 -0.4292>, rot=<0.0418 -0.0455 0.0343 -0.9975>}
 calf_twist_02_r {parent=calf_r, loc=<18.8064 0.0967 -0.0397>, rot=<0.0237 0.0095 -0.0035 -0.9997>}
 calf_twist_01_r {parent=calf_r, loc=<35.813 0.0967 -0.0397>, rot=<0.0237 0.0095 -0.0035 -0.9997>}
 deform_knee_in_base_r {parent=calf_r, loc=<-1.298 5.2957 0.0496>, rot=<0 0 0 -1>}
 deform_knee_in_r {parent=deform_knee_in_base_r, loc=<0 0 0>, rot=<0 0 0 1>}
 dyn_knee_knot_r_01 {parent=calf_r, loc=<10.661 -9.1263 2.7669>, rot=<-0.3938 0.0911 0.9138 0.0414>}
 dyn_knee_knot_r_02 {parent=calf_r, loc=<10.3955 -8.9389 -2.2904>, rot=<0.4532 -0.0854 0.8873 -0.0034>}
 thigh_twist_01_r {parent=thigh_r, loc=<21.6303 0 0>, rot=<0.0013 0 0 -1>}
 deform_groin_base_r {parent=thigh_r, loc=<2.1908 -8.1142 1.7442>, rot=<0 0 0 -1>}
 deform_groin_r {parent=deform_groin_base_r, loc=<0 0 0>, rot=<0 0 0 1>}
 deform_glute_base_l {parent=pelvis, loc=<-8.8064 -8.8222 -8.4287>, rot=<0.0275 0.0501 -0.4472 -0.8926>}
 deform_glute_l {parent=deform_glute_base_l, loc=<0 0 0>, rot=<0 0 0 1>}
 deform_glute_base_r {parent=pelvis, loc=<-8.8062 -8.8222 8.4287>, rot=<0.0501 -0.0275 0.8926 -0.4472>}
 deform_glute_r {parent=deform_glute_base_r, loc=<0 0 0>, rot=<0 0 0 1>}
 CR_cape_rotate {parent=pelvis, loc=<0 0 0>, rot=<0 0 0 1>}
 CR_cape_aimat {parent=CR_cape_rotate, loc=<0 -21.7266 0>, rot=<0 0 0 1>}
 dyn_floor_collision {parent=root, loc=<0 0 0>, rot=<0 0 0 1>}
]
variables [
 Anim Dynamics Cape {type=bool, default=False}
 RigUnit_GetTransform_2_1_Transform {type=FTransform, default=(Rotation=(X=0,Y=0,Z=0,W=1),Translation=(X=0,Y=0,Z=0),Scale3D=(X=1,Y=1,Z=1))}
 RigVMFunction_MathVectorDistance_Result {type=float, default=0}
 WearingBackbling {type=bool, default=False}
]
