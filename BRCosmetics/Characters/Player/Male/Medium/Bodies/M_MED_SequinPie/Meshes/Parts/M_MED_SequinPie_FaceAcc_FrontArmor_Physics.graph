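# Physics asset for the front leg armor part of the M_MED_SequinPie male character body, setting up capsule bodies and joints for its dynamic bones.
# A chain of five progressively larger capsules hangs from a small master body jointed to the pelvis, alongside large pelvis and thigh capsules.
# The first chain joint has wider limits, while the lower four bend mainly on swing1 with very tight swing2 and twist.
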PhysicsAsset /BRCosmetics/Characters/Player/Male/Medium/Bodies/M_MED_SequinPie/Meshes/Parts/M_MED_SequinPie_FaceAcc_FrontArmor_Physics
bodies {
 dyn_leg_armor_front_01 [{shapeType=Capsule, length=8.9688, radius=3, shapeName=dyn_leg_armor_front_01_capsule, center=<3 -2.4579 0>, rotation=<rot 0 -90 0>}]
 dyn_leg_armor_front_02 [{
   shapeType=Capsule
   length=11.0312
   radius=3.25
   shapeName=dyn_leg_armor_front_02_capsule
   center=<3 -2.3611 0>
   rotation=<rot -0.0732 -90 -0.5947>
  }]
 dyn_leg_armor_front_03 [{
   shapeType=Capsule
   length=14.0625
   radius=3.3125
   shapeName=dyn_leg_armor_front_03_capsule
   center=<3 -2.6199 0>
   rotation=<rot 0.1239 -90 -0.1335>
  }]
 dyn_leg_armor_front_04 [{shapeType=Capsule, length=17.5938, radius=3.5938, shapeName=dyn_leg_armor_front_04_capsule, center=<3 -2.4143 0>, rotation=<rot 0 -90 0>}]
 dyn_leg_armor_front_05 [{shapeType=Capsule, length=20.7188, radius=4.2812, shapeName=dyn_leg_armor_front_05_capsule, center=<3 -2.0724 0>, rotation=<rot 0 -90 0>}]
 dyn_leg_armor_front_MSTR [{shapeType=Capsule, length=0.505, radius=0.505, shapeName=dyn_leg_armor_front_MSTR_capsule, rotation=<rot 0 -90 0>}]
 pelvis [{shapeType=Capsule, length=13.7253, radius=12.0077, center=<0 -1.5685 0>}]
 thigh_l [{
   shapeType=Capsule
   length=38.005
   radius=12.8175
   shapeName=thigh_l_capsule
   center=<-19.6335 -1.1112 -0.3233>
   rotation=<rot -16.2533 -93.1192 -90.1728>
  }]
 thigh_r [{
   shapeType=Capsule
   length=38.005
   radius=12.8175
   shapeName=thigh_l_capsule
   center=<19.6335 1.1112 0.3233>
   rotation=<rot -16.2533 -93.1192 89.8272>
  }]
}
constraints [
 {
  bone1Name=thigh_l
  bone2Name=pelvis
  swing1LimitDegrees=45
  swing1Motion=Free
  swing2LimitDegrees=45
  swing2Motion=Free
  twistLimitDegrees=45
  twistMotion=Free
 }
 {
  bone1Name=dyn_leg_armor_front_MSTR
  bone2Name=pelvis
  swing1LimitDegrees=45
  swing1Motion=Free
  swing2LimitDegrees=45
  swing2Motion=Free
  twistLimitDegrees=45
  twistMotion=Free
 }
 {bone1Name=dyn_leg_armor_front_01, bone2Name=dyn_leg_armor_front_MSTR, swing1LimitDegrees=28, swing2LimitDegrees=10, twistLimitDegrees=20}
 {
  bone1Name=thigh_r
  bone2Name=pelvis
  swing1LimitDegrees=45
  swing1Motion=Free
  swing2LimitDegrees=45
  swing2Motion=Free
  twistLimitDegrees=45
  twistMotion=Free
 }
 {bone1Name=dyn_leg_armor_front_03, bone2Name=dyn_leg_armor_front_02, swing1LimitDegrees=16, swing2LimitDegrees=2, twistLimitDegrees=2}
 {bone1Name=dyn_leg_armor_front_04, bone2Name=dyn_leg_armor_front_03, swing1LimitDegrees=16, swing2LimitDegrees=2, twistLimitDegrees=2}
 {bone1Name=dyn_leg_armor_front_05, bone2Name=dyn_leg_armor_front_04, swing1LimitDegrees=16, swing2LimitDegrees=2, twistLimitDegrees=2}
 {bone1Name=dyn_leg_armor_front_02, bone2Name=dyn_leg_armor_front_01, swing1LimitDegrees=16, swing2LimitDegrees=2, twistLimitDegrees=2}
]
